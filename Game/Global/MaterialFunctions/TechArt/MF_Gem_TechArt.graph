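# Material function that adds an optional stylized gem look to incoming material attributes, toggled by the UseGem static switch.
# It darkens base color by GemDeepness, then adds a tinted glow built from the powered Fresnel input, a scatter term and sun color.
# The sun highlight uses cubed dot product of sun direction and vertex normal, avoiding per-pixel normals.
# Specular is scaled, and a selectable channel of the Alpha input masks the blend.
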MaterialFunction /Game/Global/MaterialFunctions/TechArt/MF_Gem_TechArt
graph {nodes [
  n1 = Clamp(None=Power(
    Base=FunctionInput() {inputName=Fresnel, inputType=FunctionInput_Scalar}
    Exp=ScalarParameter() {defaultValue=2, group=06_Generic Gem, parameterName=GemFresnelPower, primitiveDataIndex=3}
   ))
  n2 = GetMaterialAttributes(None=FunctionInput() {bUsePreviewValueAsDefault=true, inputName=Material, inputType=FunctionInput_MaterialAttributes}) {attributeGetTypes [#1, #2]}
  Clamp(None=@n1, Max=@n2.BaseColor)
  n3 = Constant()
  n4 = ScalarParameter() {defaultValue=0.5, group=06_Generic Gem, parameterName=GemDeepness}
  n5 = VectorParameter() {group=06_Generic Gem, parameterName=GemGlowColor, defaultValue=<rgba 1 1 1 1>}
  n6 = Clamp(None=@n1) {maxDefault=0.8, minDefault=0.2}
  n7 = MaterialFunctionCall() {materialFunction=/Game/Global/MaterialFunctions/FNCollectionParameters/MF_SunParameters_FNC}
  n8 = Multiply(A=Multiply(A=@n6, B=ScalarParameter() {defaultValue=1, group=06_Generic Gem, parameterName=GemLightScatter}) {constB=0.5}) {constB=0.2}
  n9 = VertexNormalWS()
  n10 = Multiply(A=@n1, B=Power(Base=DotProduct(A=@n7.Direction, B=@n9)) {constExponent=3})
  n11 = Multiply(A=@n6, B=Multiply(A=@n7.Color, B=Add(A=@n8, B=@n10)))
  n12 = ScalarParameter() {defaultValue=1.1, group=06_Generic Gem, parameterName=GemSpecularAdjustment}
  n13 = SetMaterialAttributes!(
   MaterialAttributes=@n2.MaterialAttributes
   BaseColor=Add(A=LinearInterpolate(A=@n2.BaseColor, B=@n3, Alpha=@n4), B=Multiply(A=@n5.RGB, B=@n11))
   Specular=Saturate(None=Multiply(A=@n2.Specular, B=@n12))
  ) {attributeSetTypes [#1, #2]}
  n14 = ChannelMaskParameter(None=FunctionInput() {bUsePreviewValueAsDefault=true, inputName=Alpha}) {group=05_Tech Art Masks, parameterName=GemMaskChannel}
  FunctionOutput(None=StaticSwitchParameter(True=BlendMaterialAttributes(A=@n2.MaterialAttributes, B=@n13, Alpha=@n14.R), False=@n2.MaterialAttributes) {group=04_Tech Art, parameterName=UseGem})
  NamedRerouteDeclaration(None=@n2.MaterialAttributes) {name=Material Attributes, nodeColor=<rgba 1 0 0.3516 1>}
  NamedRerouteUsage()
  NamedRerouteUsage()
  Saturate(None=LinearInterpolate(Alpha=@n4) {constA=1, constB=0})
 ]}
