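# Tech-art material function that optionally adds subsurface shading to incoming material attributes, meant for clear surfaces like crystal or liquids because it causes shadowing artifacts on opaque ones.
# Subsurface opacity is constant or Fresnel-driven, and scales a color optionally tinted by base color.
# A quality switch applies this shading model only at Epic, with lower tiers using a fixed-opacity fallback.
# A selectable Alpha channel masks the blend, gated by UseSubsurface.
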MaterialFunction /Game/Global/MaterialFunctions/TechArt/MF_Subsurface_TechArt
graph {nodes [
  n1 = NamedRerouteUsage()
  n2 = GetMaterialAttributes(None=FunctionInput() {inputName=Material, inputType=FunctionInput_MaterialAttributes}) {attributeGetTypes [
    #1
    #2
    #3
    #4
   ]}
  n3 = SetMaterialAttributes!(MaterialAttributes=@n2.MaterialAttributes, SubsurfaceColor=@n2.MaterialAttributes, Opacity=Constant() {r=0.5}) {attributeSetTypes [#1, #2]}
  n4 = VectorParameter() {group=06_SS Shader Model, parameterName=SS Color, defaultValue=<rgba 1 1 1 0>}
  n5 = Multiply(A=@n2.BaseColor, B=@n4.RGB)
  n6 = Power(
   Base=FunctionInput() {bUsePreviewValueAsDefault=true, inputName=Fresnel, inputType=FunctionInput_Scalar}
   Exp=ScalarParameter() {defaultValue=4, group=06_SS Shader Model, parameterName=SSFresnelExponent}
  )
  n7 = StaticSwitchParameter(
   True=LinearInterpolate(
    A=ScalarParameter() {group=06_SS Shader Model, parameterName=SSInnerStrength}
    B=ScalarParameter() {defaultValue=0.85, group=06_SS Shader Model, parameterName=SSOuterStrength}
    Alpha=Saturate(None=@n6)
   )
   False=ScalarParameter() {group=06_SS Shader Model, parameterName=SSOpacity}
  ) {group=06_SS Shader Model, parameterName=UseFresnelforSSOpacity}
  n8 = SetMaterialAttributes!(
   MaterialAttributes=NamedRerouteUsage()
   SubsurfaceColor=Multiply(A=StaticSwitchParameter(True=@n5, False=@n4.RGB) {group=06_SS Shader Model, parameterName=UseDiffuseInSSColor}, B=@n7)
   Opacity=@n7
   ShadingModel=ShadingModel() {shadingModel=MSM_Subsurface}
  ) {attributeSetTypes [#1, #2, #5]}
  n9 = FunctionInput() {bUsePreviewValueAsDefault=true, inputName=Alpha, sortPriority=1}
  n10 = StaticSwitchParameter(
   True=BlendMaterialAttributes(
    A=@n1
    B=MaterialFunctionCall!(
     Low _ PC Low and Med=@n3
     "High _ XB1, PS4"=@n3
     "Epic _ PC High, XSX, PS5"=@n8
     --------
     "Mobile Low _ PC Perf, Android"=@n3
     Mobile Med _ Switch=@n3
    ) {materialFunction=&/MF_Fortnite_QualitySwitch_MaterialAttributes}.Result
    Alpha=ChannelMaskParameter(None=@n9) {group=05_Tech Art Masks, parameterName=SubsurfaceMaskChannel}.R
   )
   False=@n1
  ) {
   desc="CAUTION: Don't use for skin or other opaque-ish surfaces, as this feature is prone to shadowing artifacts.\r\nOnly use for clear surfaces like e.g. crystal or liquids!"
   group=04_Tech Art
   parameterName=UseSubsurface
  }
  FunctionOutput(None=@n10)
  NamedRerouteDeclaration(None=@n2.MaterialAttributes) {name=Material Attributes, nodeColor=<rgba 0 0.2031 1 1>}
  Reroute(None=NamedRerouteUsage())
  Reroute(None=@n2.Normal)
 ]}
