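# Material function for the UpbeatIguana female player cosmetic body that layers shared TechArt surface effects onto incoming material attributes.
# Effects chain in order: cloth fuzz, silk, thin film, clear coat, metal LUT, gem, subsurface, advanced emissive and anisotropic.
# Named reroutes distribute alphas and shared inputs such as Fresnel, local position and a vertex-interpolated pre-skinned normal.
# Outputs the layered material and passes the Tech Art Mask through.
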MaterialFunction /BRCosmetics/Characters/Player/Female/Medium/Bodies/F_MED_UpbeatIguana/Materials/MF_UpbeatIguana_BasicTechArt
graph {nodes [
  n1 = NamedRerouteUsage()
  n2 = NamedRerouteUsage()
  n3 = NamedRerouteUsage()
  n4 = NamedRerouteUsage()
  n5 = MaterialFunctionCall!(
   Material=FunctionInput() {inputName=Material, inputType=FunctionInput_MaterialAttributes}
   Ao/Gmap/Skin=FunctionInput() {inputName=Mask Input}
   Fresnel=NamedRerouteUsage()
   Alpha=NamedRerouteUsage()
  ) {materialFunction=/Game/Global/MaterialFunctions/TechArt/MF_ClothFuzz_TechArt}
  n6 = NamedRerouteUsage()
  n7 = MaterialFunctionCall!(
   Material=MaterialFunctionCall!(
    Material Attributes=MaterialFunctionCall!(Material=@n5.Result, Alpha=@n6) {materialFunction=/Game/Global/MaterialFunctions/TechArt/MF_Silk_TechArt}.Result
    Fresnel=NamedRerouteUsage()
    Alpha=NamedRerouteUsage()
   ) {materialFunction=/Game/Global/MaterialFunctions/TechArt/MF_ThinFilm_TechArt}.Result
   Alpha=NamedRerouteUsage()
  ) {materialFunction=/Game/Global/MaterialFunctions/TechArt/MF_ClearCoat_TechArt}
  n8 = MaterialFunctionCall!(
   Fresnel=NamedRerouteUsage()
   Material=MaterialFunctionCall!(
    Fresnel=NamedRerouteUsage()
    Alpha=NamedRerouteUsage()
    Material=MaterialFunctionCall!(Alpha=@n4, Materials=@n7.Result) {materialFunction=/Game/Global/MaterialFunctions/TechArt/MF_MetalLUT_TechArt}.Result
   ) {materialFunction=/Game/Global/MaterialFunctions/TechArt/MF_Gem_TechArt}.Result
   Alpha=NamedRerouteUsage()
  ) {materialFunction=/Game/Global/MaterialFunctions/TechArt/MF_Subsurface_TechArt}
  n9 = NamedRerouteUsage()
  FunctionOutput(None=MaterialFunctionCall!(In=MaterialFunctionCall!(Local Position=@n1, Alpha=@n2, Fresnel=@n3, Material=@n8.Result, Local Normal=@n9) {materialFunction=/Game/Global/MaterialFunctions/TechArt/MF_AdvancedEmissive_TechArt}.Result) {materialFunction=/Game/Global/MaterialFunctions/TechArt/MF_Anisotropic_TechArt}.Result)
  FunctionOutput(None=NamedRerouteDeclaration(None=FunctionInput() {bUsePreviewValueAsDefault=true, inputName=Tech Art Mask}) {name=Tech Art Mask, nodeColor=<rgba 1 0.7031 0 1>}) {outputName=Tech Art Mask}
  NamedRerouteDeclaration(None=FunctionInput() {inputName=Fresnel, inputType=FunctionInput_Scalar}) {name=Fresnel, nodeColor=<rgba 0.2656 0 1 1>}
  NamedRerouteDeclaration(None=FunctionInput() {inputName=LocalPosition}) {name=Local Position, nodeColor=<rgba 0 0.7422 1 1>}
  NamedRerouteDeclaration(None=VertexInterpolator(VS=PreSkinnedNormal()).PS) {name=Local Normal, nodeColor=<rgba 0 0.2432 1 1>}
  NamedRerouteUsage()
 ]}
