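# Material function for the UpbeatIguana character that picks flipbook eye frames from an alpha-sampled head texture, with a separate frame for each eye.
# Two SubUV_Function calls sample the same texture on UV channel 1, driven by a shared SubImages grid and the left and right eye index scalar parameters.
# A step on the U coordinate at the midpoint, shared through the LeftRightEye named reroute, selects between the two results.
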MaterialFunction /BRCosmetics/Characters/Player/Female/Medium/Bodies/F_MED_UpbeatIguana/Materials/MF_FlipbookEyes
graph {nodes [
  n1 = TextureObject() {
   samplerType=SAMPLERTYPE_Alpha
   texture=/BRCosmetics/Characters/Player/Female/Medium/Bodies/F_MED_UpbeatIguana/Textures/Ash/T_UpbeatIguanaAsh_Head_FlipbookEyes
  }
  n2 = TextureCoordinate() {coordinateIndex=1}
  n3 = VectorParameter() {parameterName=SubImages, defaultValue=<rgba 3 3 0 1>}
  n4 = ScalarParameter() {parameterName=flipbook_face_L_eye_index}
  n5 = ScalarParameter() {parameterName=flipbook_face_R_eye_index}
  FunctionOutput(None=LinearInterpolate(
    A=MaterialFunctionCall!(Texture=@n1, UVs=@n2, SubImages=@n3.RGB, Frame=@n4) {materialFunction=/Engine/Functions/Engine_MaterialFunctions03/Texturing/SubUV_Function}.RGB
    B=MaterialFunctionCall(Texture=@n1, UVs=@n2, SubImages=@n3.RGB, Frame=@n5) {materialFunction=&SubUV_Function}.RGB
    Alpha=NamedRerouteUsage()
   ))
  n6 = TextureCoordinate() {coordinateIndex=1}
  NamedRerouteDeclaration(None=ComponentMask(None=Step(Y=@n6) {constX=0.5}) {r=true}) {name=LeftRightEye, nodeColor=<rgba 0.2422 1 0 1>}
 ]}
