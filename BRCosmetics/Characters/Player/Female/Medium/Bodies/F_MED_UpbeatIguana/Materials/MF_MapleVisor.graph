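# Material function for the UpbeatIguana Maple visor that turns part of an input material into an animated monitor screen with a glowing logo.
# An Alpha input drives the override of base color and emissive, with ScreenColor and LogoColor inputs supplying tints.
# Panned static, a timed sweep and distance-faded scanlines build the screen effect.
# The logo uses BumpOffset parallax and UV distortion, while a hue-shift blend is left unconnected.
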MaterialFunction /BRCosmetics/Characters/Player/Female/Medium/Bodies/F_MED_UpbeatIguana/Materials/MF_MapleVisor
graph {nodes [
  n1 = GetMaterialAttributes(None=FunctionInput() {inputName=Material, inputType=FunctionInput_MaterialAttributes}) {attributeGetTypes [#1]}
  n2 = NamedRerouteUsage()
  n3 = LinearInterpolate(A=NamedRerouteUsage(), Alpha=NamedRerouteUsage()) {constB=0}
  n4 = NamedRerouteUsage()
  FunctionOutput(None=SetMaterialAttributes!(
    MaterialAttributes=@n1.MaterialAttributes
    BaseColor=LinearInterpolate(A=@n1.BaseColor, Alpha=@n2) {constB=0}
    EmissiveColor=LinearInterpolate(B=@n3, Alpha=@n4)
   ) {attributeSetTypes [#1, #2]})
  n5 = Multiply(
   A=NamedRerouteUsage()
   B=LinearInterpolate(A=ScalarParameter() {defaultValue=0.8, group=MonitorStatic, parameterName=ScreenBrightness}, Alpha=NamedRerouteUsage()) {constA=1, constB=0}
  )
  LinearInterpolate(
   A=MaterialFunctionCall(Hue Shift Percentage=ScalarParameter() {defaultValue=0.1, group=MonitorStatic, parameterName=FresnelHueShift}, Texture=@n5) {materialFunction=&HueShift}.Result
   B=@n5
   Alpha=NamedRerouteUsage()
  )
  NamedRerouteDeclaration(None=FunctionInput() {inputName=Alpha, inputType=FunctionInput_Scalar, sortPriority=1}) {name=Alpha, nodeColor=<rgba 0 0.7891 1 1>}
  NamedRerouteDeclaration(None=Multiply(A=Fresnel() {baseReflectFraction=0, exponent=1}, B=ScalarParameter() {defaultValue=1, group=MonitorStatic, parameterName=FresnelIntensity})) {name=Fresnel, nodeColor=<rgba 0 1 0.6953 1>}
  n6 = NamedRerouteUsage()
  n7 = Constant2Vector() {g=-0.05, r=6}
  n8 = NamedRerouteUsage()
  n9 = TextureSample(UVs=Panner(Coordinate=@n6, Speed=Multiply(A=@n7, B=@n8)) {speedX=6, speedY=-0.05}) {
   samplerSource=SSM_Wrap_WorldGroupSettings
   samplerType=SAMPLERTYPE_Masks
   texture=/Game/Missions/Primary/RepairTheShelter/Textures/Cinematic_Screen_Static
  }
  n10 = Constant2Vector() {g=-0.03, r=-3}
  n11 = NamedRerouteUsage()
  n12 = TextureSample(UVs=Panner(Coordinate=@n6, Speed=Multiply(A=@n10, B=@n11)) {speedX=-3, speedY=-0.03}) {
   samplerSource=SSM_Wrap_WorldGroupSettings
   samplerType=SAMPLERTYPE_Masks
   texture=&Cinematic_Screen_Static
  }
  n13 = Multiply(A=@n9.G, B=@n12.R)
  n14 = ScalarParameter() {defaultValue=0.07, group=MonitorStatic, parameterName=Static UV Distort}
  NamedRerouteDeclaration(None=AppendVector(A=Multiply(A=@n13, B=@n14), B=Constant())) {name=StaticUVDistrotion, nodeColor=<rgba 1 0.5625 0 1>}
  n15 = NamedRerouteUsage()
  n16 = NamedRerouteUsage()
  n17 = VectorParameter() {group=MonitorStatic, parameterName=LogoUVScale}
  n18 = Add(
   A=Multiply(A=TextureCoordinate(), B=ComponentMask(None=@n17.RGB) {g=true, r=true})
   B=ComponentMask(None=VectorParameter() {group=MonitorStatic, parameterName=LogoUVOffset}.RGB) {g=true, r=true}
  )
  n19 = ScalarParameter() {defaultValue=-0.05, group=MonitorStatic, parameterName=ParallexIntensity}
  n20 = TextureSampleParameter2D(UVs=Add(A=Add(A=@n15, B=@n16), B=BumpOffset(Coordinate=@n18, Height=@n19))) {
   group=MonitorStatic
   parameterName=LogoTexture
   samplerSource=SSM_Clamp_WorldGroupSettings
   samplerType=SAMPLERTYPE_Alpha
   texture=/BRCosmetics/Characters/Player/Female/Medium/Bodies/F_MED_UpbeatIguana/Textures/Maple/T_UpbeatIguana_Maple_Logo
  }
  n21 = Multiply(A=Multiply(A=NamedRerouteUsage(), B=@n20.R), B=ScalarParameter() {defaultValue=2, group=MonitorStatic, parameterName=LogoBrightness})
  NamedRerouteDeclaration(None=@n21) {name=Debug, nodeColor=<rgba 1 0 0.1406 1>}
  n22 = Multiply(A=ComponentMask(None=NamedRerouteUsage()) {g=true}) {constB=0.75}
  n23 = Multiply(A=Time()) {constB=-0.2}
  n24 = Power(Base=Frac(None=Add(A=@n22, B=@n23))) {constExponent=20}
  n25 = ScalarParameter() {defaultValue=0.005, group=MonitorStatic, parameterName=Sweep UV Distort}
  NamedRerouteDeclaration(None=AppendVector(A=Multiply(A=@n24, B=@n25) {constB=0.02}, B=Constant())) {name=SweepUVDistortion, nodeColor=<rgba 0.0781 1 0 1>}
  NamedRerouteDeclaration(None=FunctionInput() {inputName=LogoColor, sortPriority=1, previewValue=<0 1 0 1>}) {name=LogoColor, nodeColor=<rgba 0.9922 1 0 1>}
  NamedRerouteDeclaration(None=ScalarParameter() {defaultValue=1, group=MonitorStatic, parameterName=StaticSpeed}) {name=StaticSpeed, nodeColor=<rgba 0 1 0.6484 1>}
  n26 = VectorParameter() {group=MonitorStatic, parameterName=UVScale, defaultValue=<rgba 1 1 0 0>}
  NamedRerouteDeclaration(None=Multiply(A=TextureCoordinate(), B=ComponentMask(None=@n26.RGB) {g=true, r=true})) {name=TexCoord, nodeColor=<rgba 1 0 0.4453 1>}
  NamedRerouteDeclaration(None=Multiply(A=@n24, B=ScalarParameter() {defaultValue=0.2, group=MonitorStatic, parameterName=Sweep Add}) {constB=0.2}) {name=Sweep, nodeColor=<rgba 0.2422 0 1 1>}
  NamedRerouteDeclaration(None=FunctionInput() {inputName=ScreenColor, sortPriority=1, previewValue=<1 0 0 1>}) {name=ScreenColor, nodeColor=<rgba 0 0.6719 1 1>}
  n27 = NamedRerouteUsage()
  n28 = ConstantBiasScale(None=Frac(None=Panner(Coordinate=@n27) {speedY=-1})) {bias=-0.5, scale=2}
  n29 = ComponentMask(None=Multiply(A=@n28, B=@n28)) {g=true}
  n30 = ScalarParameter() {defaultValue=0.5, group=MonitorStatic, parameterName=Scanline Strength}
  NamedRerouteDeclaration(None=Add(A=Multiply(A=@n29, B=@n30) {constB=0.1})) {name=Scanlines, nodeColor=<rgba 0 0.1562 1 1>}
  n31 = ScalarParameter() {defaultValue=50, group=MonitorStatic, parameterName=ScanlineDistanceFadeRange}
  n32 = Multiply(A=ScalarParameter() {defaultValue=200, group=MonitorStatic, parameterName=ScanlineDistanceFadeStart}) {constB=-1}
  NamedRerouteDeclaration(None=LinearInterpolate(
    A=MaterialFunctionCall!(ScanLine_Tiling_1, Scanline_PannerSpeed_1, ScanLine_Tiling, Scanline_PannerSpeed, UVs) {materialFunction=/Game/Weapons/WeaponSkins/MaterialFunction/ScanLines}.Result
    B=Constant() {r=0.5}
    Alpha=MaterialFunctionCall!(Blend Range=@n31, Start Offset=@n32) {materialFunction=/Engine/Functions/Engine_MaterialFunctions03/Particles/Distance_Blend}.Result
   )) {name=Scanlines_02, nodeColor=<rgba 0.4766 1 0 1>}
  n33 = NamedRerouteUsage()
  n34 = NamedRerouteUsage()
  n35 = NamedRerouteUsage()
  n36 = Add(A=@n21, B=Multiply(A=Add(A=@n34, B=@n35)) {constB=0.35})
  NamedRerouteDeclaration(None=Multiply(A=NamedRerouteUsage(), B=LinearInterpolate(A=@n5, B=@n33, Alpha=@n36)) {constA=1}) {name=OUT_Emissive, nodeColor=<rgba 1 0 0.7969 1>}
  n37 = Constant2Vector() {g=0.1, r=5}
  n38 = NamedRerouteUsage()
  n39 = TextureSample(UVs=Panner(Coordinate=@n6, Speed=Multiply(A=@n37, B=@n38)) {speedX=5, speedY=0.1}) {
   samplerSource=SSM_Wrap_WorldGroupSettings
   samplerType=SAMPLERTYPE_Masks
   texture=&Cinematic_Screen_Static
  }
  n40 = Multiply(A=@n39.B) {constB=0.1}
  NamedRerouteDeclaration(None=Multiply(A=Add(A=@n40, B=@n13), B=ScalarParameter() {defaultValue=6, group=MonitorStatic, parameterName=Static Scale})) {name=Static, nodeColor=<rgba 0 1 0.8828 1>}
  NamedRerouteUsage()
 ]}
