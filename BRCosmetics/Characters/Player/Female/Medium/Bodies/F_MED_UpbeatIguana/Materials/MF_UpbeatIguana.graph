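# Material function for the UpbeatIguana cosmetic set, recoloring its body, accessories and heads through static switches, including backpack and pickaxe variants.
# A customization mask splits zones with SphereMasks on banded values and RGB channels into primary, secondary and tertiary colors, layered with patterns and a placeable logo.
# It overrides incoming material attributes, adding skin fresnel, blinking parallax backpack eyes, Ash/Maple face effects and vertex-color-driven WPO that hides Birch eye poses.
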MaterialFunction /BRCosmetics/Characters/Player/Female/Medium/Bodies/F_MED_UpbeatIguana/Materials/MF_UpbeatIguana
graph {nodes [
  Constant3Vector() {constant=<rgba 0.2385 0.5 0.0182 1>}
  Constant3Vector() {constant=<rgba 0.01 0.03 0.298 1>}
  n1 = GetMaterialAttributes(None=FunctionInput() {inputType=FunctionInput_MaterialAttributes}) {attributeGetTypes [
    #1
    #2
    #3
    #4
   ]}
  n2 = SetMaterialAttributes!(
   MaterialAttributes=@n1.MaterialAttributes
   BaseColor=NamedRerouteUsage()
   Specular=NamedRerouteUsage()
   Roughness=NamedRerouteUsage()
   EmissiveColor=NamedRerouteUsage()
   Normal=NamedRerouteUsage()
   WorldPositionOffset=NamedRerouteUsage()
  ) {attributeSetTypes [
    #1
    #2
    #3
    #5
    #4
    #6
   ]}
  n3 = NamedRerouteUsage()
  n4 = NamedRerouteUsage()
  n5 = NamedRerouteUsage()
  FunctionOutput(None=StaticSwitchParameter(
    True=MaterialFunctionCall!(Material=@n2, ScreenColor=@n3, LogoColor=@n4, Alpha=@n5) {materialFunction=&/MF_MapleVisor}.Result
    False=@n2
   ) {group=UpbeatIguana, parameterName=UseMapleHead})
  FunctionOutput(None=NamedRerouteUsage()) {outputName=OpacityMask, sortPriority=1}
  FunctionOutput(None=NamedRerouteUsage()) {outputName=UpperBody_PrimaryColor, sortPriority=2}
  FunctionOutput(None=NamedRerouteUsage()) {outputName=TilingPatternMask, sortPriority=3}
  n6 = FunctionInput() {bUsePreviewValueAsDefault=true, inputName=TechArtMask, sortPriority=2}
  n7 = AppendVector(A=@n6, B=Constant())
  n8 = TextureSampleParameter2D() {
   group=UpbeatIguana
   parameterName=CustomizationMask
   samplerType=SAMPLERTYPE_Masks
   texture=/BRCosmetics/Characters/Player/Female/Medium/Bodies/F_MED_UpbeatIguana/Textures/T_UpbeatIguana_Body_FX
  }
  n9 = MaterialFunctionCall(Float3=NamedRerouteDeclaration(None=StaticSwitchParameter(True=@n7, False=@n8.RGBA) {group=UpbeatIguana, parameterName=UseTechArtMaskForCustomization}) {name=CustomizationMask, nodeColor=<rgba 0 1 0.8828 1>}) {materialFunction=&BreakOutFloat3Components}
  n10 = Add(A=@n9.R, B=@n9.R)
  n11 = NamedRerouteUsage()
  n12 = TextureCoordinate()
  n13 = Multiply(A=NamedRerouteUsage(), B=Step(Y=ComponentMask(None=@n12) {r=true}) {constX=0.52, constY=0.52})
  n14 = Add(A=Multiply(A=NamedRerouteUsage(), B=Add(A=@n9.R, B=@n9.R)), B=Add(A=Multiply(A=@n10, B=@n11), B=Multiply(A=@n9.R, B=@n13)))
  n15 = NamedRerouteUsage()
  n16 = NamedRerouteDeclaration(None=Saturate(None=Subtract(A=@n14, B=@n15))) {name=ClothMask, nodeColor=<rgba 1 0 0.3984 1>}
  FunctionOutput(None=StaticSwitchParameter(
    True=MaterialFunctionCall(X=@n16, Y=@n9.G, Z=@n9.G) {materialFunction=&MakeFloat3}.Result
    False=NamedRerouteUsage()
   ) {group=UpbeatIguana, parameterName=UseBody}) {outputName=TechArtMask, sortPriority=3}
  n17 = NamedRerouteUsage()
  n18 = SmoothStep(Value=Distance(A=@n17, B=Constant3Vector() {constant=<rgba 0 0.2708 0.1139 1>})) {constMax=0.6}
  n19 = MaterialFunctionCall(
   Hue Shift Percentage=LinearInterpolate(A=ScalarParameter() {defaultValue=-0.075, parameterName=UI_SkinHueShift}, Alpha=@n18) {constB=0}
   Texture=@n17
  ) {desc=Hue shift on green skin, materialFunction=&HueShift}
  n20 = Add(A=@n19.Result, B=ScalarParameter() {defaultValue=0.075, parameterName=DarkSkinAdd})
  LinearInterpolate(A=@n20, B=@n19.Result) {desc=Brighter on black skin}
  MaterialFunctionCall(Number=@n18) {materialFunction=&DebugScalarValues}
  Multiply(A=@n19.Result) {constB=5}
  n21 = ScalarParameter() {defaultValue=0.05, group=UpbeatIguana, parameterName=CustomizationMaskThreshold}
  NamedRerouteDeclaration(None=SphereMask(A=NamedRerouteUsage(), B=Constant() {r=0.2}, Radius=@n21) {attenuationRadius=0.1}) {name=UpperBody_Mask, nodeColor=<rgba 0.5703 1 0 1>}
  NamedRerouteDeclaration(None=SphereMask(A=NamedRerouteUsage(), B=Constant() {r=0.4}, Radius=@n21) {attenuationRadius=0.1}) {name=PantsSleves_Mask, nodeColor=<rgba 0.5703 1 0 1>}
  n22 = TextureSampleParameter2D() {
   group=UpbeatIguana
   parameterName=UI_TilingPatternTexture
   samplerType=SAMPLERTYPE_LinearColor
   texture=/BRCosmetics/Characters/Player/Female/Medium/Bodies/F_MED_UpbeatIguana/Textures/T_UpbeatIguana_Body_Pattern
  }
  n23 = ScalarParameter() {group=UpbeatIguana, parameterName=UI_LogoStyle}
  n24 = Multiply(A=NamedRerouteUsage(), B=ScalarParameter() {defaultValue=1, group=UpbeatIguana, parameterName=UI_UsePattern_Upper})
  n25 = Multiply(A=NamedRerouteUsage(), B=ScalarParameter() {defaultValue=1, group=UpbeatIguana, parameterName=UI_UsePattern_Lower})
  NamedRerouteDeclaration(None=LinearInterpolate(B=LinearInterpolate(A=@n22.G, B=@n22.G, Alpha=@n23), Alpha=Add(A=@n24, B=@n25))) {name=TilingPattern, nodeColor=<rgba 1 0.2578 0 1>}
  n26 = FunctionInput() {bUsePreviewValueAsDefault=true, inputName=M_Map, sortPriority=1}
  NamedRerouteDeclaration(None=SmoothStep(Value=ComponentMask(None=@n26) {b=true}) {constMax=0.5, desc=Sharpen the mask edge a bit}) {name=SkinMask, nodeColor=<rgba 0 0.7187 1 1>}
  n27 = NamedRerouteUsage()
  n28 = LinearInterpolate(B=@n27, Alpha=ScalarParameter() {defaultValue=0.5, group=UpbeatIguana, parameterName=BaseColorGradientIntensity}) {constA=1}
  n29 = LinearInterpolate(A=@n28, Alpha=NamedRerouteUsage()) {constB=0.75, desc=Removed chest line when logo is on}
  n30 = NamedRerouteUsage()
  n31 = StaticSwitchParameter(True=NamedRerouteUsage(), False=NamedRerouteUsage()) {group=UpbeatIguana, parameterName=UseAcc}
  n32 = NamedRerouteUsage()
  n33 = Add(A=@n32, B=NamedRerouteUsage())
  n34 = LinearInterpolate(
   A=@n27
   B=StaticSwitchParameter(True=Multiply(A=@n29, B=@n30), False=Multiply(A=@n28, B=@n31)) {group=UpbeatIguana, parameterName=UseBody}
   Alpha=StaticSwitchParameter(True=Saturate(None=@n33), False=@n32) {group=UpbeatIguana, parameterName=UsePickaxe}
  )
  n35 = Constant3Vector() {constant=<rgba 0.6867 0.6867 0.6867 1>}
  n36 = SphereMask(A=VertexInterpolator(VS=LocalPosition() {localOrigin=InstancePreSkinning}.XYZ).PS, B=Constant3Vector() {constant=<rgba 0 0 -60 0>}) {attenuationRadius=20}
  n37 = Multiply(A=@n36, B=NamedRerouteUsage())
  n38 = Multiply(A=Multiply(A=OneMinus(None=@n37), B=@n36)) {constB=0.4}
  NamedRerouteDeclaration(None=StaticSwitchParameter(True=LinearInterpolate(A=@n34, B=@n35, Alpha=@n38), False=@n34) {group=UpbeatIguana, parameterName=UsePickaxe}) {name=BaseColor_Appearance, nodeColor=<rgba 1 0.1406 0 1>}
  NamedRerouteDeclaration(None=@n1.Specular) {name=Specular_IN, nodeColor=<rgba 0.4297 0 1 1>}
  NamedRerouteDeclaration(None=@n1.Roughness) {name=Roughness_IN, nodeColor=<rgba 0.4297 0 1 1>}
  n39 = NamedRerouteUsage()
  n40 = LinearInterpolate(
   A=@n39
   B=LinearInterpolate(A=@n39, B=ScalarParameter() {defaultValue=0.5, group=UpbeatIguana, parameterName=UI_MaskSkinSpecular}, Alpha=NamedRerouteUsage())
   Alpha=ScalarParameter() {group=UpbeatIguana, parameterName=UI_UseMaskSkin}
  )
  n41 = ScalarParameter() {defaultValue=0.5, parameterName=UI_Logo_Specular}
  n42 = Add(A=NamedRerouteUsage(), B=NamedRerouteUsage())
  n43 = NamedRerouteUsage()
  n44 = Saturate(None=Add(A=NamedRerouteUsage(), B=Multiply(A=@n42, B=@n43) {desc=TODO}))
  NamedRerouteDeclaration(None=StaticSwitchParameter(True=LinearInterpolate(A=@n40, B=@n41, Alpha=@n44), False=NamedRerouteUsage()) {group=UpbeatIguana, parameterName=UseBody}) {name=Specular_OUT, nodeColor=<rgba 1 0 0.2109 1>}
  n45 = NamedRerouteUsage()
  n46 = LinearInterpolate(
   A=@n45
   B=LinearInterpolate(A=@n45, B=ScalarParameter() {defaultValue=0.15, group=UpbeatIguana, parameterName=UI_MaskSkinRoughness}, Alpha=NamedRerouteUsage())
   Alpha=ScalarParameter() {group=UpbeatIguana, parameterName=UI_UseMaskSkin}
  )
  n47 = ScalarParameter() {defaultValue=0.5, parameterName=UI_Logo_Roughness}
  n48 = NamedRerouteUsage()
  NamedRerouteDeclaration(None=StaticSwitchParameter(True=LinearInterpolate(A=@n46, B=@n47, Alpha=@n48), False=NamedRerouteUsage()) {group=UpbeatIguana, parameterName=UseBody}) {name=Roughness_OUT, nodeColor=<rgba 0 0.5312 1 1>}
  n49 = ComponentMask(None=NamedRerouteUsage()) {b=true, g=true, r=true}
  n50 = MaterialFunctionCall(Float3=@n49) {materialFunction=&BreakOutFloat3Components}
  n51 = Add(A=@n50.R, B=@n50.R)
  NamedRerouteDeclaration(None=SmoothStep(Value=Add(A=@n51, B=@n50.B)) {constMax=0.1}) {name=CustomizationAlpha, nodeColor=<rgba 0.6406 1 0 1>}
  NamedRerouteDeclaration(None=SphereMask(A=NamedRerouteUsage(), B=Constant() {r=0.5}, Radius=@n21) {attenuationRadius=0.1}) {name=Shoes_Mask, nodeColor=<rgba 0.5703 1 0 1>}
  n52 = VectorParameter() {group=UpbeatIguana, parameterName=UI_UpperBody_PrimaryColor, sortPriority=50, defaultValue=<rgba 1 0 0 1>}
  n53 = ScalarParameter() {defaultValue=0.065, group=UpbeatIguana, parameterName=UI_PatternOpacity}
  n54 = Constant() {desc=Black subtract, r=0.05}
  n55 = Desaturation(None=@n52.RGB, Fraction=Constant() {r=1})
  n56 = Constant() {desc=Black radius, r=0.15}
  n57 = LinearInterpolate(B=@n54, Alpha=SphereMask(A=@n55, B=Constant() {r=0.02}, Radius=@n56) {attenuationRadius=0.1, hardnessPercent=0}) {constB=0.045}
  n58 = Constant() {desc=White add, r=0.275}
  n59 = Constant() {desc=White radius, r=0.26}
  n60 = SphereMask(A=@n55, B=Constant() {r=0.7}, Radius=@n59) {attenuationRadius=0.2, hardnessPercent=0}
  n61 = Saturate(None=Add(A=Subtract(A=@n53, B=@n57), B=LinearInterpolate(B=@n58, Alpha=@n60) {constB=0.3}))
  n62 = NamedRerouteUsage()
  n63 = LinearInterpolate(
   A=NamedRerouteUsage()
   B=LinearInterpolate(
    A=NamedRerouteDeclaration(None=@n52.RGB) {name=UpperBody_PrimaryColor, nodeColor=<rgba 1 0.1875 0 1>}
    Alpha=Multiply(A=@n61, B=@n62) {constA=0.2, constB=0.2}
   )
   Alpha=Multiply(A=NamedRerouteUsage(), B=NamedRerouteUsage())
  )
  n64 = VectorParameter() {group=UpbeatIguana, parameterName=UI_UpperBody_SecondaryColor, sortPriority=50, defaultValue=<rgba 0 1 0 1>}
  n65 = Desaturation(None=@n64.RGB, Fraction=Constant() {r=1})
  n66 = SphereMask(A=@n65, B=Constant() {r=0.02}, Radius=@n56) {attenuationRadius=0.1, hardnessPercent=0}
  n67 = Constant() {r=0.7}
  n68 = Add(
   A=Subtract(A=@n53, B=LinearInterpolate(B=@n54, Alpha=@n66) {constB=0.045})
   B=LinearInterpolate(B=@n58, Alpha=SphereMask(A=@n65, B=@n67, Radius=@n59) {attenuationRadius=0.2, hardnessPercent=0}) {constB=0.3}
  )
  n69 = LinearInterpolate(A=@n64.RGB, Alpha=Multiply(A=Saturate(None=@n68), B=NamedRerouteUsage()) {constA=0.2, constB=0.2})
  n70 = Multiply(A=NamedRerouteUsage(), B=NamedRerouteUsage())
  n71 = VectorParameter() {group=UpbeatIguana, parameterName=UI_UpperBody_TertiaryColor, sortPriority=50, defaultValue=<rgba 0 0 1 1>}
  n72 = NamedRerouteUsage()
  n73 = ScalarParameter() {defaultValue=1, group=UpbeatIguana, parameterName=UI_UseLogo_Upper}
  n74 = Multiply(A=Add(A=NamedRerouteUsage(), B=Multiply(A=@n72, B=@n73)), B=NamedRerouteUsage())
  n75 = VectorParameter() {group=UpbeatIguana, parameterName=UI_PantsSleves_PrimaryColor, sortPriority=51, defaultValue=<rgba 1 0 0 1>}
  n76 = Desaturation(None=@n75.RGB, Fraction=Constant() {r=1})
  n77 = LinearInterpolate(B=@n54, Alpha=SphereMask(A=@n76, B=Constant() {r=0.02}, Radius=@n56) {attenuationRadius=0.1, hardnessPercent=0}) {constB=0.035}
  n78 = SphereMask(A=@n76, B=Constant() {r=0.7}, Radius=@n59) {attenuationRadius=0.2, hardnessPercent=0}
  n79 = Saturate(None=Add(A=Subtract(A=@n53, B=@n77), B=LinearInterpolate(B=@n58, Alpha=@n78) {constB=0.3}))
  n80 = NamedRerouteUsage()
  n81 = NamedRerouteUsage()
  n82 = LinearInterpolate(
   A=LinearInterpolate(
    A=LinearInterpolate(A=@n63, B=@n69, Alpha=@n70)
    B=@n71.RGB
    Alpha=NamedRerouteDeclaration(None=@n74) {name=StripesMask, nodeColor=<rgba 0 0.4844 1 1>}
   )
   B=LinearInterpolate(A=@n75.RGB, Alpha=Multiply(A=@n79, B=@n80) {constA=0.2, constB=0.2})
   Alpha=Multiply(A=NamedRerouteUsage(), B=@n81)
  )
  n83 = VectorParameter() {group=UpbeatIguana, parameterName=UI_PantsSleves_SecondaryColor, sortPriority=51, defaultValue=<rgba 0 1 0 1>}
  n84 = Multiply(A=NamedRerouteUsage(), B=@n81)
  n85 = VectorParameter() {group=UpbeatIguana, parameterName=UI_PantsSleves_TertiaryColor, sortPriority=51, defaultValue=<rgba 0 0 1 1>}
  n86 = Add(
   A=NamedRerouteUsage()
   B=Multiply(A=NamedRerouteUsage(), B=ScalarParameter() {defaultValue=1, group=UpbeatIguana, parameterName=UI_UseLogo_Lower})
  )
  n87 = VectorParameter() {group=UpbeatIguana, parameterName=UI_Shoes_PrimaryColor, sortPriority=52, defaultValue=<rgba 1 0 0 1>}
  n88 = NamedRerouteUsage()
  n89 = LinearInterpolate(
   A=LinearInterpolate(A=LinearInterpolate(A=@n82, B=@n83.RGB, Alpha=@n84), B=@n85.RGB, Alpha=Multiply(A=@n86, B=@n81))
   B=@n87.RGB
   Alpha=Multiply(A=NamedRerouteUsage(), B=@n88)
  )
  n90 = VectorParameter() {group=UpbeatIguana, parameterName=UI_Shoes_SecondaryColor, sortPriority=52, defaultValue=<rgba 0 1 0 1>}
  n91 = Multiply(A=NamedRerouteUsage(), B=@n88)
  n92 = VectorParameter() {group=UpbeatIguana, parameterName=UI_Shoes_TertiaryColor, sortPriority=52, defaultValue=<rgba 0 0 1 1>}
  n93 = NamedRerouteUsage()
  n94 = VectorParameter() {group=UpbeatIguana, parameterName=UI_Bag_PrimaryColor, sortPriority=53, defaultValue=<rgba 1 0 0 1>}
  n95 = NamedRerouteUsage()
  n96 = LinearInterpolate(
   A=LinearInterpolate(A=LinearInterpolate(A=@n89, B=@n90.RGB, Alpha=@n91), B=@n92.RGB, Alpha=Multiply(A=@n93, B=@n88))
   B=@n94.RGB
   Alpha=Multiply(A=NamedRerouteUsage(), B=@n95)
  )
  n97 = VectorParameter() {group=UpbeatIguana, parameterName=UI_Bag_SecondaryColor, sortPriority=53, defaultValue=<rgba 0 1 0 1>}
  n98 = Multiply(A=NamedRerouteUsage(), B=@n95)
  n99 = VectorParameter() {group=UpbeatIguana, parameterName=UI_Bag_TertiaryColor, sortPriority=53, defaultValue=<rgba 0 0 1 1>}
  n100 = NamedRerouteUsage()
  NamedRerouteDeclaration(None=LinearInterpolate(A=LinearInterpolate(A=@n96, B=@n97.RGB, Alpha=@n98), B=@n99.RGB, Alpha=Multiply(A=@n100, B=@n95))) {name=BodyColor, nodeColor=<rgba 0 1 0.9063 1>}
  n101 = VectorParameter() {group=UpbeatIguana, parameterName=UI_FaceAcc_PrimaryColor, sortPriority=54, defaultValue=<rgba 1 0 0 1>}
  n102 = NamedRerouteDeclaration(None=@n101.RGB) {name=Maple_ScreenColor, nodeColor=<rgba 0 1 0.2734 1>}
  n103 = LinearInterpolate(
   A=NamedRerouteUsage()
   B=@n102
   Alpha=Multiply(A=NamedRerouteUsage(), B=ScalarParameter() {defaultValue=1, group=UpbeatIguana, parameterName=UI_MakeupColorOpacity_R})
  )
  n104 = VectorParameter() {group=UpbeatIguana, parameterName=UI_FaceAcc_SecondaryColor, sortPriority=54, defaultValue=<rgba 0 1 0 1>}
  n105 = Multiply(A=NamedRerouteUsage(), B=ScalarParameter() {defaultValue=1, group=UpbeatIguana, parameterName=UI_MakeupColorOpacity_G})
  n106 = VectorParameter() {group=UpbeatIguana, parameterName=UI_FaceAcc_TertiaryColor, sortPriority=55, defaultValue=<rgba 0 0 1 1>}
  n107 = Clamp(None=Multiply(A=@n106.RGB, B=ScalarParameter() {defaultValue=1, group=UpbeatIguana, parameterName=UI_FaceAcc_BaseColorMulti})) {maxDefault=0.9}
  n108 = Multiply(A=NamedRerouteUsage(), B=ScalarParameter() {defaultValue=1, group=UpbeatIguana, parameterName=UI_MakeupColorOpacity_B})
  n109 = Saturate(None=Add(A=@n108, B=NamedRerouteUsage()))
  NamedRerouteDeclaration(None=LinearInterpolate(
    A=LinearInterpolate(A=@n103, B=@n104.RGB, Alpha=@n105)
    B=@n107
    Alpha=StaticSwitchParameter(True=@n109, False=@n108) {group=UpbeatIguana, parameterName=UseBirchHead}
   )) {name=FaceAccColor, nodeColor=<rgba 1 0.5859 0 1>}
  n110 = TextureCoordinate() {coordinateIndex=2}
  n111 = TextureCoordinate() {coordinateIndex=1}
  n112 = VectorParameter() {group=UpbeatIguana, parameterName=UI_LogoLoc_Stripes, defaultValue=<rgba 0 -0.16 0 1>}
  n113 = VectorParameter() {group=UpbeatIguana, parameterName=UI_LogoLoc_Trefoil, defaultValue=<rgba 0 0.36 0 1>}
  n114 = OneMinus(None=@n23)
  MaterialExpressionAdd_3 = Add(
   A=MaterialFunctionCall(
    UVs=StaticSwitchParameter(True=@n110, False=@n111) {group=UpbeatIguana, parameterName=UseBackpack}
    Texture Scale=ScalarParameter() {defaultValue=1, group=UpbeatIguana, parameterName=UI_LogoScale}
   ) {materialFunction=&ScaleUVsByCenter}.UVs
   B=LinearInterpolate(A=ComponentMask(None=@n112.RGB) {g=true, r=true}, B=ComponentMask(None=@n113.RGB) {g=true, r=true}, Alpha=@n114)
  )
  n115 = Step(Y=ComponentMask(None=@MaterialExpressionAdd_3) {g=true}) {constX=0.4, constY=0.5}
  n116 = OneMinus(None=@n115)
  n117 = Constant() {r=0.5}
  MaterialExpressionConstant2Vector_2 = Constant2Vector() {g=0.65, r=0.61}
  MaterialExpressionLinearInterpolate_35 = LinearInterpolate(A=Constant2Vector() {g=0.35, r=0.5}, B=Constant2Vector() {g=0.8, r=0.5}, Alpha=@n114)
  n118 = Custom!(uv=@MaterialExpressionAdd_3, size=@MaterialExpressionConstant2Vector_2, offset=@MaterialExpressionLinearInterpolate_35) {
   code="return (uv / size + offset) - offset / size;"
   description=ScaleUVwithOffset
   outputType=CMOT_Float2
   inputs [
    {
     inputName=uv
     input {expression=&self:MaterialExpressionAdd_3}
    }
    {
     inputName=size
     input {
      expression=&self:MaterialExpressionConstant2Vector_2
      mask=1
      maskG=1
      maskR=1
     }
    }
    {
     inputName=offset
     input {expression=&self:MaterialExpressionLinearInterpolate_35}
    }
   ]
  }
  n119 = LinearInterpolate(A=@MaterialExpressionAdd_3, B=@n118, Alpha=NamedRerouteUsage())
  n120 = MaterialFunctionCall(Float2=@n119) {materialFunction=&BreakOutFloat2Components}
  n121 = SphereMask(A=@n117, B=@n120.R) {attenuationRadius=0.5}
  n122 = SphereMask(A=@n120.G, B=@n117) {attenuationRadius=0.5}
  n123 = TextureSampleParameter2D(UVs=@n119) {
   group=UpbeatIguana
   parameterName=UI_LogoTexture
   samplerType=SAMPLERTYPE_Alpha
   texture=/BRCosmetics/Characters/Player/Female/Medium/Bodies/F_MED_UpbeatIguana/Textures/T_UpbeatIguana_Body_Logo
  }
  n124 = Multiply(A=Multiply(A=LinearInterpolate(A=@n115, B=@n116, Alpha=@n114), B=Multiply(A=@n121, B=@n122)), B=@n123.R)
  n125 = NamedRerouteUsage()
  n126 = ScalarParameter() {defaultValue=1, group=UpbeatIguana, parameterName=UI_UseLogo_Upper}
  n127 = NamedRerouteUsage()
  n128 = ScalarParameter() {defaultValue=1, group=UpbeatIguana, parameterName=UI_UseLogo_Lower}
  n129 = LinearInterpolate(B=@n124, Alpha=Add(A=Multiply(A=@n125, B=@n126), B=Multiply(A=@n127, B=@n128)))
  NamedRerouteDeclaration(None=Saturate(None=StaticSwitchParameter(True=@n124, False=@n129) {group=UpbeatIguana, parameterName=UseBackpack})) {name=LogoMask, nodeColor=<rgba 0.2656 0 1 1>}
  n130 = Clamp(None=ScalarParameter() {group=UpbeatIguana, parameterName=UI_SkinColorRGBSwitch}) {maxDefault=2}
  NamedRerouteDeclaration(None=MaterialFunctionCall(1=@n102, 2=@n104.RGB, 3=@n107, A=Divide(A=@n130)) {materialFunction=&Lerp_Multiple_Float3}.Lerp 3 Inputs) {name=SkinColor, nodeColor=<rgba 1 0.0703 0 1>}
  n131 = Constant3Vector() {constant=<rgba 0.2992 0.8 0 1>}
  n132 = NamedRerouteUsage()
  n133 = Multiply(
   A=@n107
   B=MaterialFunctionCall() {materialFunction=&/MF_FlipbookEyes}.Result
  )
  n134 = Constant()
  NamedRerouteDeclaration(None=StaticSwitchParameter(True=Multiply(A=@n131, B=@n132), False=StaticSwitchParameter(True=@n133, False=@n134) {group=UpbeatIguana, parameterName=UseAshHead}) {group=UpbeatIguana, parameterName=UseBackpack}) {name=Emissive_OUT, nodeColor=<rgba 1 0 0.9375 1>}
  n135 = VectorParameter() {group=UpbeatIguana, parameterName=UI_Acc_PrimaryColor, sortPriority=54, defaultValue=<rgba 1 0 0 1>}
  n136 = NamedRerouteUsage()
  n137 = Add(A=@n136, B=NamedRerouteUsage())
  n138 = LinearInterpolate(
   A=NamedRerouteUsage()
   B=Multiply(A=@n135.RGB, B=ScalarParameter() {defaultValue=1, group=UpbeatIguana, parameterName=UI_Acc_PrimaryColor_Multi})
   Alpha=StaticSwitchParameter(True=Saturate(None=@n137), False=@n136) {group=UpbeatIguana, parameterName=UsePickaxe}
  )
  n139 = VectorParameter() {group=UpbeatIguana, parameterName=UI_Acc_SecondaryColor, sortPriority=54, defaultValue=<rgba 0 1 0 1>}
  n140 = NamedRerouteUsage()
  n141 = VectorParameter() {group=UpbeatIguana, parameterName=UI_Acc_TertiaryColor, sortPriority=55, defaultValue=<rgba 0 0 1 1>}
  n142 = NamedRerouteUsage()
  n143 = NamedRerouteUsage()
  NamedRerouteDeclaration(None=LinearInterpolate(A=LinearInterpolate(A=@n138, B=@n139.RGB, Alpha=@n140), B=@n141.RGB, Alpha=Add(A=@n142, B=@n143))) {name=AccColor, nodeColor=<rgba 1 0.5859 0 1>}
  NamedRerouteDeclaration() {name=Debug, nodeColor=<rgba 0 1 0.8125 1>}
  NamedRerouteDeclaration(None=@n1.Normal) {name=Normal_IN, nodeColor=<rgba 1 0 0.0469 1>}
  NamedRerouteDeclaration(None=SphereMask(A=NamedRerouteUsage(), B=Constant() {r=0.8}, Radius=@n21) {attenuationRadius=0.1}) {name=Bag_Mask, nodeColor=<rgba 0.5703 1 0 1>}
  n144 = NamedRerouteUsage()
  n145 = Constant3Vector() {constant=<rgba 0 0 1 0>}
  n146 = NamedRerouteUsage()
  n147 = NamedRerouteUsage()
  n148 = Saturate(None=Add(A=ComponentMask(None=@n147) {r=true}, B=ComponentMask(None=@n147) {g=true}))
  n149 = Multiply(A=NamedRerouteUsage(), B=SmoothStep(Value=Multiply(A=@n146, B=@n148)) {constMin=0.8})
  NamedRerouteDeclaration(None=StaticSwitchParameter(True=LinearInterpolate(A=@n144, B=@n145, Alpha=@n149), False=@n144) {group=UpbeatIguana, parameterName=UseBody}) {name=Normal_OUT, nodeColor=<rgba 0 0.3672 1 1>}
  NamedRerouteDeclaration(None=@n107) {name=MapleLogoColor, nodeColor=<rgba 0.5938 0 1 1>}
  n150 = TextureObject() {
   samplerType=SAMPLERTYPE_Alpha
   texture=/BRCosmetics/Characters/Player/Female/Medium/Bodies/F_MED_UpbeatIguana/Textures/Ash/T_UpbeatIguanaAsh_Head_FlipbookEyes
  }
  n151 = MaterialFunctionCall(
   UVs=BumpOffset(
    Coordinate=TextureCoordinate() {coordinateIndex=1}
    Height=ScalarParameter() {defaultValue=-2, group=UpbeatIguana, parameterName=UI_Backpack_EyesParallaxHeight}
   )
   Texture Scale=Constant2Vector() {g=0.8, r=0.8}
  ) {materialFunction=&ScaleUVsByCenter}
  n152 = ScalarParameter() {defaultValue=0.25, group=UpbeatIguana, parameterName=UI_BackpackBlinkSpeed}
  MaterialExpressionMultiply_34 = Multiply(A=Time(), B=@n152) {constB=0.25}
  n153 = Custom!(t=@MaterialExpressionMultiply_34) {
   code="return frac(2*sin(0.5*t) + 0.15*sin(4.5*t));"
   description=RandomBlink
   inputs [{
     inputName=t
     input {expression=&self:MaterialExpressionMultiply_34}
    }]
  }
  n154 = Step(Y=@n153) {constX=0.06}
  n155 = MaterialFunctionCall!(
   Texture=@n150
   UVs=Add(A=@n151.UVs, B=Constant2Vector() {g=0.225})
   SubImages=Constant2Vector() {g=3, r=3}
   Frame=LinearInterpolate(A=Constant(), B=Constant() {r=3}, Alpha=NamedRerouteDeclaration(None=@n154) {name=EyeBlink, nodeColor=<rgba 1 0.0703 0 1>})
  ) {materialFunction=/Engine/Functions/Engine_MaterialFunctions03/Texturing/SubUV_Function}
  NamedRerouteDeclaration(None=Multiply(A=@n151.0-1 mask, B=ComponentMask(None=@n155.RGB) {r=true})) {name=Backpack_EyesMask, nodeColor=<rgba 0.7344 1 0 1>}
  n156 = Constant() {r=1}
  n157 = VertexColor()
  n158 = MaterialFunctionCall!(
   Hide Element Toggle=@n156
   VertexColorChannel=@n157.G
   Vertex Value to hide=ScalarParameter() {defaultValue=0.25, group=BirchFlipbookEyes, parameterName=R_blink_pose_value}
  ) {materialFunction=/Game/Packages/Fortress_SharedMaterials/Base_Material_Functions/VertexColorMask}
  n159 = MaterialFunctionCall(
   Hide Element Toggle=@n156
   VertexColorChannel=@n157.G
   Vertex Value to hide=ScalarParameter() {defaultValue=0.1, group=BirchFlipbookEyes, parameterName=default_pose_value}
  ) {materialFunction=&VertexColorMask}
  n160 = Max(A=@n158.Result, B=@n159.Result)
  NamedRerouteDeclaration(None=@n160) {name=BirchEyesMask, nodeColor=<rgba 0.8281 0 1 1>}
  n161 = ScalarParameter() {group=BirchFlipbookEyes, parameterName=R_blink_pose, sortPriority=0}
  n162 = ScalarParameter() {defaultValue=0.1, group=BirchFlipbookEyes, parameterName=Threshold_R_blink_pose}
  n163 = Subtract(A=OneMinus(None=If(A=@n161, B=@n162, "A > B"=@n158.Result, "A == B"=@n159.Result, "A < B"=@n159.Result)), B=OneMinus(None=@n160))
  n164 = Subtract(A=ObjectPositionWS(), B=WorldPosition().XYZ)
  NamedRerouteDeclaration(None=StaticSwitchParameter(True=Multiply(A=@n163, B=@n164), False=Constant()) {group=UpbeatIguana, parameterName=UseBirchHead}) {name=WPO_OUT, nodeColor=<rgba 0 1 0.5313 1>}
  NamedRerouteDeclaration(None=ComponentMask(None=NamedRerouteUsage()) {a=true}) {name=CustomizationMask_A, nodeColor=<rgba 0 0.2031 1 1>}
  NamedRerouteDeclaration(None=@n49) {name=RGB, nodeColor=<rgba 1 0 0.0469 1>}
  NamedRerouteDeclaration(None=@n36) {name=SphereMask, nodeColor=<rgba 0 0.9531 1 1>}
  NamedRerouteDeclaration(None=@n6) {name=TechArtMask_0, nodeColor=<rgba 1 0 0.7969 1>}
  n165 = ScalarParameter() {group=UpbeatIguana, parameterName=UI_MaskSmoothStep_RG_Min}
  n166 = ScalarParameter() {defaultValue=0.5, group=UpbeatIguana, parameterName=UI_MaskSmoothStep_RG_Max}
  NamedRerouteDeclaration(None=SmoothStep(Min=@n165, Max=@n166, Value=@n50.R) {constMax=0.9, constMin=0.5}) {name=Reds, nodeColor=<rgba 1 0 0.0243 1>}
  NamedRerouteDeclaration(None=SmoothStep(Min=@n165, Max=@n166, Value=@n50.G) {constMax=0.9, constMin=0.5}) {name=Greens, nodeColor=<rgba 0 1 0 1>}
  NamedRerouteDeclaration(None=SmoothStep(
    Min=ScalarParameter() {defaultValue=0.5, group=UpbeatIguana, parameterName=UI_MaskSmoothStep_B_Min}
    Max=ScalarParameter() {defaultValue=1, group=UpbeatIguana, parameterName=UI_MaskSmoothStep_B_Max}
    Value=@n50.B
   ) {constMax=0.9, constMin=0.5}) {name=Blues, nodeColor=<rgba 0 0 1 1>}
  NamedRerouteDeclaration(None=@n1.BaseColor) {name=BaseColor_IN, nodeColor=<rgba 0 0.0156 1 1>}
  n167 = NamedRerouteUsage()
  n168 = ScalarParameter() {defaultValue=0.5, group=UpbeatIguana, parameterName=UI_SkinFresnelExp}
  n169 = ScalarParameter() {defaultValue=0.3, group=UpbeatIguana, parameterName=UI_SkinFresnelBias}
  n170 = LinearInterpolate(
   A=@n167
   B=LinearInterpolate(B=@n20, Alpha=Fresnel(ExponentIn=@n168, BaseReflectFractionIn=@n169) {baseReflectFraction=0.1, exponent=1})
   Alpha=NamedRerouteUsage()
  )
  n171 = ScalarParameter() {group=UpbeatIguana, parameterName=UI_UseMaskSkin}
  NamedRerouteDeclaration(None=StaticSwitchParameter(True=LinearInterpolate(A=@n167, B=@n170, Alpha=@n171), False=@n167) {group=UpbeatIguana, parameterName=UseBody}) {name=BaseColor_OUT, nodeColor=<rgba 0.3359 1 0 1>}
  NamedRerouteUsage()
  ScalarParameter() {group=UpbeatIguana, parameterName=UI_PatternColorInfluence}
  SmoothStep(Value=Distance(A=@n17, B=Constant3Vector() {constant=<rgba 0.02 0.02 0.02 1>})) {constMax=0.3}
  Step(Y=MaterialFunctionCall(Speed=@n152) {materialFunction=&DebugTimeSine}.Result) {constX=0.5, desc=0.01}
 ]}
