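# Translucent parent material for the glassy clear sleeves on the UpbeatIguana female character body cosmetic.
# MF_UpbeatIguana supplies customization color and masks, which drive a Fresnel inner-to-outer opacity blend that gains pattern-masked transparency when the primary color nears white.
# A sort-translucency function processes opacity to fix sorting.
# The lit attributes then pass through thin film, optional camera dither (UseCameraDither switch), cloth-sim WPO and TechArt overrides.
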Material /BRCosmetics/Characters/Player/Female/Medium/Bodies/F_MED_UpbeatIguana/Materials/M_UpbeatIguana_ClearSleeves_Parent
graph {
 properties {
  allowTranslucentCustomDepthWrites=true
  bAutomaticallySetUsageInEditor=false
  bEnableResponsiveAA=true
  bScreenSpaceReflections=true
  bUseMaterialAttributes=true
  bUsedWithMeshDeformer=true
  bUsedWithMorphTargets=true
  bUsedWithSkeletalMesh=true
  blendMode=BLEND_Translucent
  ditherOpacityMask=true
  pixelDepthOffsetMode=PDOM_Legacy
  translucencyDirectionalLightingIntensity=5
  translucencyLightingMode=TLM_SurfacePerPixelLighting
  translucencyPass=MTP_BeforeDOF
  usedShadingModels=Default Lit
 }
 nodes [
  n1 = MaterialFunctionCall!(
   In=SetMaterialAttributes!(
    MaterialAttributes
    Base Color=TextureSampleParameter2D() {group=UpbeatIguana, parameterName=Diffuse, sortPriority=10, texture=&T_BlacksRGB}.RGB
   ) {attributeSetTypes [#1]}
   M_Map
   TechArtMask
  ) {desc=customization, materialFunction=&/MF_UpbeatIguana}
  n2 = ScalarParameter() {defaultValue=0.7, group=UpbeatIguana, parameterName=Fresnel Inner Opacity}
  n3 = Multiply(A=@n2, B=Constant() {r=0.4}) {constB=0.5}
  n4 = Constant3Vector() {constant=<rgba 1 1 1 0>}
  n5 = NamedRerouteDeclaration(None=Fresnel(ExponentIn=ScalarParameter() {defaultValue=1, group=UpbeatIguana, parameterName=Fresnel Exponent})) {name=Fresnel, nodeColor=<rgba 0 1 0.2734 1>}
  n6 = ScalarParameter() {defaultValue=1, group=UpbeatIguana, parameterName=Glass Thickness}
  n7 = LinearInterpolate(
   A=LinearInterpolate(
    A=@n2
    B=LinearInterpolate(A=@n3, B=@n2, Alpha=@n1.TilingPatternMask) {desc=More transparent for white}
    Alpha=SphereMask(A=@n4, B=@n1.UpperBody_PrimaryColor) {attenuationRadius=0.5, hardnessPercent=0}
   )
   B=ScalarParameter() {defaultValue=0.9, group=UpbeatIguana, parameterName=Fresnel Outer Opacity}
   Alpha=Clamp(None=Multiply(A=@n5, B=@n6))
  )
  n8 = TextureSampleParameter2D() {
   group=UpbeatIguana
   parameterName=Normals
   samplerSource=SSM_Wrap_WorldGroupSettings
   samplerType=SAMPLERTYPE_Normal
   texture=/Engine/EngineMaterials/FlatNormal
  }
  n9 = SetMaterialAttributes!(
   MaterialAttributes=@n1.Result
   Specular=ScalarParameter() {defaultValue=0.5, group=UpbeatIguana, parameterName=Specular}
   Roughness=ScalarParameter() {group=UpbeatIguana, parameterName=Roughness}
   Metallic=ScalarParameter() {defaultValue=0.25, group=UpbeatIguana, parameterName=Metallic}
   Emissive Color
   Opacity=MaterialFunctionCall!(Opacity=LinearInterpolate(A=@n7, Alpha=@n1.OpacityMask)) {desc=Fix transparency sorting, materialFunction=/Game/Global/Materials/MaterialGym/Materials/SortedOutline/MF_SortTranslucency}.Result
   Normal=@n8.RGB
   Opacity Mask
  ) {attributeSetTypes [
    #2
    #3
    #4
    #5
    #6
    #7
    #8
   ]}
  n10 = MaterialFunctionCall!(Material Attributes=@n9, Fresnel=NamedRerouteUsage(), Alpha=Constant() {r=1}) {materialFunction=/Game/Global/MaterialFunctions/TechArt/MF_ThinFilm_TechArt}
  n11 = MaterialFunctionCall!(In=@n10.Result) {materialFunction=/Game/Packages/Fortress_SharedMaterials/Base_Material_Functions/BaseCustomizationFunctions/MF_FN_Vehicle_CameraDither}
  MaterialFunctionCall(In=MaterialFunctionCall!(
    Material=StaticSwitchParameter(True=@n11.Result, False=@n10.Result) {group=00_Glass, parameterName=UseCameraDither, sortPriority=250}
    SkinFX_Mask=Constant()
   ) {materialFunction=/Game/Global/MaterialFunctions/MF_ClothSimWPO_v2}.Result) {materialFunction=&MF_TechArtOverrides}
  ScalarParameter() {defaultValue=0.65, group=UpbeatIguana, parameterName=Opacity}
 ]
}
