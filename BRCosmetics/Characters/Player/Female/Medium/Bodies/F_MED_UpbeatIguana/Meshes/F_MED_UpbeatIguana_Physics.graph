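# Physics asset for the F_MED_UpbeatIguana female cosmetic body, covering a purse and the body parts around it.
# Capsule bodies sit on two spine bones, the left upper arm and two purse bones.
# Constraints chain upper arm and purse base to the upper spine, and that to the lower spine, all with free rotation.
# Only the joint from purse base to dynamic purse bone differs, with tight swing on one axis and zero twist limit.
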PhysicsAsset /BRCosmetics/Characters/Player/Female/Medium/Bodies/F_MED_UpbeatIguana/Meshes/F_MED_UpbeatIguana_Physics
bodies {
 base_purse [{shapeType=Capsule, length=0.505, radius=0.505, shapeName=base_purse_capsule}]
 dyn_purse_01 [{shapeType=Capsule, length=3, radius=3, center=<15.2144 0 0>, rotation=<rot 90 0 0>}]
 spine_01 [{shapeType=Capsule, length=3.5217, radius=10.4783, center=<0 1.3155 0>, rotation=<rot 90 0 0>}]
 spine_04 [{shapeType=Capsule, length=3.5217, radius=11.2907, center=<-5.4408 0.5814 0>, rotation=<rot 90 0 -14.9999>}]
 upperarm_l [{shapeType=Capsule, length=20.3484, radius=5.1763, center=<14.9082 -2.1401 -2.0696>, rotation=<rot 90 -179.5524 -174.5528>}]
}
constraints [
 {
  bone1Name=upperarm_l
  bone2Name=spine_04
  swing1LimitDegrees=45
  swing1Motion=Free
  swing2LimitDegrees=45
  swing2Motion=Free
  twistLimitDegrees=45
  twistMotion=Free
 }
 {
  bone1Name=spine_04
  bone2Name=spine_01
  swing1LimitDegrees=45
  swing1Motion=Free
  swing2LimitDegrees=45
  swing2Motion=Free
  twistLimitDegrees=45
  twistMotion=Free
 }
 {
  bone1Name=base_purse
  bone2Name=spine_04
  swing1LimitDegrees=45
  swing1Motion=Free
  swing2LimitDegrees=45
  swing2Motion=Free
  twistLimitDegrees=45
  twistMotion=Free
 }
 {bone1Name=dyn_purse_01, bone2Name=base_purse, swing1LimitDegrees=5, swing2LimitDegrees=45, twistLimitDegrees=0}
]
