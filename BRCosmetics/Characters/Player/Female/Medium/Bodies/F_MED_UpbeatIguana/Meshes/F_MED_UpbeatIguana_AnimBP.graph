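# Animation Blueprint for the F_MED_UpbeatIguana cosmetic body, defining secondary motion for sleeves, straps, purse and zips.
# Cached stages run Control Rig, constraints and state-driven bone offsets into RigidBody and AnimDynamics, adding world-space passes in the front end.
# Output Pose, however, links directly to Copy Pose From Mesh, bypassing that chain.
# The event graph computes gravity override, flags a specific face part for backpack offset, and drives WPO cloth wind materials.
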AnimBlueprint /BRCosmetics/Characters/Player/Female/Medium/Bodies/F_MED_UpbeatIguana/Meshes/F_MED_UpbeatIguana_AnimBP
target_skeleton=&/F_MED_UpbeatIguana_Skeleton
graphs [
 AnimGraph {nodes [
   n1 = Get Override World Gravity() -> (Override World Gravity:Vector)
   n2 = Anim Dynamics - Bone: dyn_Sleeve_front_r!(ComponentPose=Local To Component(LocalPose=Use cached pose 'Cache02'()), Alpha=0.4, GravityOverride:Vector=@n1, ExternalForce="0,0,0") -> (Pose) {
    angularDampingOverride=1
    angularSpringConstant=20000
    bAngularSpring=true
    bOverrideAngularDamping=true
    bUseGravityOverride=true
    gravityScale=0.35
    simulationSpace=RootRelative
    alphaBoolBlend {blendOption=Linear}
    boundBone {boneName=dyn_Sleeve_front_r}
    componentAppliedLinearAccClamp=<3000 3000 3000>
    componentLinearAccScale=<0.3 0.3 0.3>
    componentLinearVelScale=<0.3 0.3 0.3>
    gravityOverride {z=-980}
    physicsBodyDefinitions [{
      boundBone {boneName=dyn_Sleeve_front_r}
      boxExtents=<20 30 30>
      constraintSetup {angularLimitsMax=<10 10 10>, angularLimitsMin=<-10 -10 -10>, angularTarget=<1 0 0>}
      localJointOffset=<0.3 0 -2>
     }]
   }
   n3 = Anim Dynamics - Bone: dyn_Sleeve_under_r!(ComponentPose=@n2, Alpha=0.4, GravityOverride:Vector=@n1, ExternalForce="0,0,0") -> (Pose) {
    angularDampingOverride=1
    angularSpringConstant=20000
    bAngularSpring=true
    bOverrideAngularDamping=true
    bUseGravityOverride=true
    gravityScale=0.35
    simulationSpace=RootRelative
    alphaBoolBlend {blendOption=Linear}
    boundBone {boneName=dyn_Sleeve_under_r}
    componentAppliedLinearAccClamp=<3000 3000 3000>
    componentLinearAccScale=<0.3 0.3 0.3>
    componentLinearVelScale=<0.3 0.3 0.3>
    gravityOverride {z=-980}
    physicsBodyDefinitions [{
      boundBone {boneName=dyn_Sleeve_under_r}
      boxExtents=<20 30 30>
      constraintSetup {angularLimitsMax=<10 10 10>, angularLimitsMin=<-10 -10 -10>, angularTarget=<1 0 0>}
      localJointOffset=<0.3 0 -2>
     }]
   }
   n4 = Anim Dynamics - Bone: dyn_Sleeve_back_r!(ComponentPose=@n3, Alpha=0.4, GravityOverride:Vector=@n1, ExternalForce="0,0,0") -> (Pose) {
    angularDampingOverride=1
    angularSpringConstant=20000
    bAngularSpring=true
    bOverrideAngularDamping=true
    bUseGravityOverride=true
    gravityScale=0.35
    simulationSpace=RootRelative
    alphaBoolBlend {blendOption=Linear}
    boundBone {boneName=dyn_Sleeve_back_r}
    componentAppliedLinearAccClamp=<3000 3000 3000>
    componentLinearAccScale=<0.3 0.3 0.3>
    componentLinearVelScale=<0.3 0.3 0.3>
    gravityOverride {z=-980}
    physicsBodyDefinitions [{
      boundBone {boneName=dyn_Sleeve_back_r}
      boxExtents=<20 30 30>
      constraintSetup {angularLimitsMax=<10 10 10>, angularLimitsMin=<-10 -10 -10>, angularTarget=<1 0 0>}
      localJointOffset=<0.3 0 -2>
     }]
   }
   n5 = Anim Dynamics - Bone: dyn_Sleeve_top_r!(ComponentPose=@n4, Alpha=0.4, GravityOverride:Vector=@n1, ExternalForce="0,0,0") -> (Pose) {
    angularDampingOverride=1
    angularSpringConstant=20000
    bAngularSpring=true
    bOverrideAngularDamping=true
    bUseGravityOverride=true
    gravityScale=0.35
    simulationSpace=RootRelative
    alphaBoolBlend {blendOption=Linear}
    boundBone {boneName=dyn_Sleeve_top_r}
    componentAppliedLinearAccClamp=<3000 3000 3000>
    componentLinearAccScale=<0.3 0.3 0.3>
    componentLinearVelScale=<0.3 0.3 0.3>
    gravityOverride {z=-980}
    physicsBodyDefinitions [{
      boundBone {boneName=dyn_Sleeve_top_r}
      boxExtents=<20 30 30>
      constraintSetup {angularLimitsMax=<10 10 10>, angularLimitsMin=<-10 -10 -10>, angularTarget=<1 0 0>}
      localJointOffset=<0.3 0 -2>
     }]
   }
   n6 = Anim Dynamics - Bone: dyn_Sleeve_top_l!(ComponentPose=@n5, Alpha=0.4, GravityOverride:Vector=@n1, ExternalForce="0,0,0") -> (Pose) {
    angularDampingOverride=1
    angularSpringConstant=20000
    bAngularSpring=true
    bOverrideAngularDamping=true
    bUseGravityOverride=true
    gravityScale=0.35
    simulationSpace=RootRelative
    alphaBoolBlend {blendOption=Linear}
    boundBone {boneName=dyn_Sleeve_top_l}
    componentAppliedLinearAccClamp=<3000 3000 3000>
    componentLinearAccScale=<0.3 0.3 0.3>
    componentLinearVelScale=<0.3 0.3 0.3>
    gravityOverride {z=-980}
    physicsBodyDefinitions [{
      boundBone {boneName=dyn_Sleeve_top_l}
      boxExtents=<20 30 30>
      constraintSetup {angularLimitsMax=<10 10 10>, angularLimitsMin=<-10 -10 -10>, angularTarget=<1 0 0>}
      localJointOffset=<0.3 0 -2>
     }]
   }
   n7 = Anim Dynamics - Bone: dyn_Sleeve_back_l!(ComponentPose=@n6, Alpha=0.4, GravityOverride:Vector=@n1, ExternalForce="0,0,0") -> (Pose) {
    angularDampingOverride=1
    angularSpringConstant=20000
    bAngularSpring=true
    bOverrideAngularDamping=true
    bUseGravityOverride=true
    gravityScale=0.35
    simulationSpace=RootRelative
    alphaBoolBlend {blendOption=Linear}
    boundBone {boneName=dyn_Sleeve_back_l}
    componentAppliedLinearAccClamp=<3000 3000 3000>
    componentLinearAccScale=<0.3 0.3 0.3>
    componentLinearVelScale=<0.3 0.3 0.3>
    gravityOverride {z=-980}
    physicsBodyDefinitions [{
      boundBone {boneName=dyn_Sleeve_back_l}
      boxExtents=<20 30 30>
      constraintSetup {angularLimitsMax=<10 10 10>, angularLimitsMin=<-10 -10 -10>, angularTarget=<1 0 0>}
      localJointOffset=<0.3 0 -2>
     }]
   }
   n8 = Anim Dynamics - Bone: dyn_Sleeve_front_l!(ComponentPose=@n7, Alpha=0.4, GravityOverride:Vector=@n1, ExternalForce="0,0,0") -> (Pose) {
    angularDampingOverride=1
    angularSpringConstant=20000
    bAngularSpring=true
    bOverrideAngularDamping=true
    bUseGravityOverride=true
    gravityScale=0.35
    simulationSpace=RootRelative
    alphaBoolBlend {blendOption=Linear}
    boundBone {boneName=dyn_Sleeve_front_l}
    componentAppliedLinearAccClamp=<3000 3000 3000>
    componentLinearAccScale=<0.3 0.3 0.3>
    componentLinearVelScale=<0.3 0.3 0.3>
    gravityOverride {z=-980}
    physicsBodyDefinitions [{
      boundBone {boneName=dyn_Sleeve_front_l}
      boxExtents=<20 30 30>
      constraintSetup {angularLimitsMax=<10 10 10>, angularLimitsMin=<-10 -10 -10>, angularTarget=<1 0 0>}
      localJointOffset=<0.3 0 -2>
     }]
   }
   n9 = Anim Dynamics - Bone: dyn_Sleeve_under_l!(ComponentPose=@n8, Alpha=0.4, GravityOverride:Vector=@n1, ExternalForce="0,0,0") -> (Pose) {
    angularDampingOverride=1
    angularSpringConstant=20000
    bAngularSpring=true
    bOverrideAngularDamping=true
    bUseGravityOverride=true
    gravityScale=0.35
    simulationSpace=RootRelative
    alphaBoolBlend {blendOption=Linear}
    boundBone {boneName=dyn_Sleeve_under_l}
    componentAppliedLinearAccClamp=<3000 3000 3000>
    componentLinearAccScale=<0.3 0.3 0.3>
    componentLinearVelScale=<0.3 0.3 0.3>
    gravityOverride {z=-980}
    physicsBodyDefinitions [{
      boundBone {boneName=dyn_Sleeve_under_l}
      boxExtents=<20 30 30>
      constraintSetup {angularLimitsMax=<10 10 10>, angularLimitsMin=<-10 -10 -10>, angularTarget=<1 0 0>}
      localJointOffset=<0.3 0 -2>
     }]
   }
   n10 = Get bIsInFrontEnd()
   n11 = Anim Dynamics - Bone: dyn_Sleeve_front_r!(ComponentPose=@n9, bAlphaBoolEnabled:bool=@n10, GravityOverride=@n1, ExternalForce="0,0,0") -> (Pose) {
    alphaInputType=Bool
    angularDampingOverride=1
    angularSpringConstant=20000
    bAngularSpring=true
    bOverrideAngularDamping=true
    bUseGravityOverride=true
    gravityScale=0.35
    simulationSpace=World
    alphaBoolBlend {blendOption=Linear}
    boundBone {boneName=dyn_Sleeve_front_r}
    componentAppliedLinearAccClamp=<3000 3000 3000>
    componentLinearAccScale=<0.3 0.3 0.3>
    componentLinearVelScale=<0.3 0.3 0.3>
    gravityOverride {z=-980}
    physicsBodyDefinitions [{
      boundBone {boneName=dyn_Sleeve_front_r}
      boxExtents=<20 30 30>
      constraintSetup {angularLimitsMax=<10 10 10>, angularLimitsMin=<-10 -10 -10>, angularTarget=<1 0 0>}
      localJointOffset=<0.3 0 -2>
     }]
   }
   n12 = Anim Dynamics - Bone: dyn_Sleeve_under_r!(ComponentPose=@n11, bAlphaBoolEnabled:bool=@n10, GravityOverride=@n1, ExternalForce="0,0,0") -> (Pose) {
    alphaInputType=Bool
    angularDampingOverride=1
    angularSpringConstant=20000
    bAngularSpring=true
    bOverrideAngularDamping=true
    bUseGravityOverride=true
    gravityScale=0.35
    simulationSpace=World
    alphaBoolBlend {blendOption=Linear}
    boundBone {boneName=dyn_Sleeve_under_r}
    componentAppliedLinearAccClamp=<3000 3000 3000>
    componentLinearAccScale=<0.3 0.3 0.3>
    componentLinearVelScale=<0.3 0.3 0.3>
    gravityOverride {z=-980}
    physicsBodyDefinitions [{
      boundBone {boneName=dyn_Sleeve_under_r}
      boxExtents=<20 30 30>
      constraintSetup {angularLimitsMax=<10 10 10>, angularLimitsMin=<-10 -10 -10>, angularTarget=<1 0 0>}
      localJointOffset=<0.3 0 -2>
     }]
   }
   n13 = Anim Dynamics - Bone: dyn_Sleeve_back_r!(ComponentPose=@n12, bAlphaBoolEnabled:bool=@n10, GravityOverride=@n1, ExternalForce="0,0,0") -> (Pose) {
    alphaInputType=Bool
    angularDampingOverride=1
    angularSpringConstant=20000
    bAngularSpring=true
    bOverrideAngularDamping=true
    bUseGravityOverride=true
    gravityScale=0.35
    simulationSpace=World
    alphaBoolBlend {blendOption=Linear}
    boundBone {boneName=dyn_Sleeve_back_r}
    componentAppliedLinearAccClamp=<3000 3000 3000>
    componentLinearAccScale=<0.3 0.3 0.3>
    componentLinearVelScale=<0.3 0.3 0.3>
    gravityOverride {z=-980}
    physicsBodyDefinitions [{
      boundBone {boneName=dyn_Sleeve_back_r}
      boxExtents=<20 30 30>
      constraintSetup {angularLimitsMax=<10 10 10>, angularLimitsMin=<-10 -10 -10>, angularTarget=<1 0 0>}
      localJointOffset=<0.3 0 -2>
     }]
   }
   n14 = Anim Dynamics - Bone: dyn_Sleeve_top_r!(ComponentPose=@n13, bAlphaBoolEnabled:bool=@n10, GravityOverride=@n1, ExternalForce="0,0,0") -> (Pose) {
    alphaInputType=Bool
    angularDampingOverride=1
    angularSpringConstant=20000
    bAngularSpring=true
    bOverrideAngularDamping=true
    bUseGravityOverride=true
    gravityScale=0.35
    simulationSpace=World
    alphaBoolBlend {blendOption=Linear}
    boundBone {boneName=dyn_Sleeve_top_r}
    componentAppliedLinearAccClamp=<3000 3000 3000>
    componentLinearAccScale=<0.3 0.3 0.3>
    componentLinearVelScale=<0.3 0.3 0.3>
    gravityOverride {z=-980}
    physicsBodyDefinitions [{
      boundBone {boneName=dyn_Sleeve_top_r}
      boxExtents=<20 30 30>
      constraintSetup {angularLimitsMax=<10 10 10>, angularLimitsMin=<-10 -10 -10>, angularTarget=<1 0 0>}
      localJointOffset=<0.3 0 -2>
     }]
   }
   n15 = Anim Dynamics - Bone: dyn_Sleeve_top_l!(ComponentPose=@n14, bAlphaBoolEnabled:bool=@n10, GravityOverride=@n1, ExternalForce="0,0,0") -> (Pose) {
    alphaInputType=Bool
    angularDampingOverride=1
    angularSpringConstant=20000
    bAngularSpring=true
    bOverrideAngularDamping=true
    bUseGravityOverride=true
    gravityScale=0.35
    simulationSpace=World
    alphaBoolBlend {blendOption=Linear}
    boundBone {boneName=dyn_Sleeve_top_l}
    componentAppliedLinearAccClamp=<3000 3000 3000>
    componentLinearAccScale=<0.3 0.3 0.3>
    componentLinearVelScale=<0.3 0.3 0.3>
    gravityOverride {z=-980}
    physicsBodyDefinitions [{
      boundBone {boneName=dyn_Sleeve_top_l}
      boxExtents=<20 30 30>
      constraintSetup {angularLimitsMax=<10 10 10>, angularLimitsMin=<-10 -10 -10>, angularTarget=<1 0 0>}
      localJointOffset=<0.3 0 -2>
     }]
   }
   n16 = Anim Dynamics - Bone: dyn_Sleeve_back_l!(ComponentPose=@n15, bAlphaBoolEnabled:bool=@n10, GravityOverride=@n1, ExternalForce="0,0,0") -> (Pose) {
    alphaInputType=Bool
    angularDampingOverride=1
    angularSpringConstant=20000
    bAngularSpring=true
    bOverrideAngularDamping=true
    bUseGravityOverride=true
    gravityScale=0.35
    simulationSpace=World
    alphaBoolBlend {blendOption=Linear}
    boundBone {boneName=dyn_Sleeve_back_l}
    componentAppliedLinearAccClamp=<3000 3000 3000>
    componentLinearAccScale=<0.3 0.3 0.3>
    componentLinearVelScale=<0.3 0.3 0.3>
    gravityOverride {z=-980}
    physicsBodyDefinitions [{
      boundBone {boneName=dyn_Sleeve_back_l}
      boxExtents=<20 30 30>
      constraintSetup {angularLimitsMax=<10 10 10>, angularLimitsMin=<-10 -10 -10>, angularTarget=<1 0 0>}
      localJointOffset=<0.3 0 -2>
     }]
   }
   n17 = Anim Dynamics - Bone: dyn_Sleeve_front_l!(ComponentPose=@n16, bAlphaBoolEnabled:bool=@n10, GravityOverride=@n1, ExternalForce="0,0,0") -> (Pose) {
    alphaInputType=Bool
    angularDampingOverride=1
    angularSpringConstant=20000
    bAngularSpring=true
    bOverrideAngularDamping=true
    bUseGravityOverride=true
    gravityScale=0.35
    simulationSpace=World
    alphaBoolBlend {blendOption=Linear}
    boundBone {boneName=dyn_Sleeve_front_l}
    componentAppliedLinearAccClamp=<3000 3000 3000>
    componentLinearAccScale=<0.3 0.3 0.3>
    componentLinearVelScale=<0.3 0.3 0.3>
    gravityOverride {z=-980}
    physicsBodyDefinitions [{
      boundBone {boneName=dyn_Sleeve_front_l}
      boxExtents=<20 30 30>
      constraintSetup {angularLimitsMax=<10 10 10>, angularLimitsMin=<-10 -10 -10>, angularTarget=<1 0 0>}
      localJointOffset=<0.3 0 -2>
     }]
   }
   n18 = Anim Dynamics - Bone: dyn_Sleeve_under_l!(ComponentPose=@n17, bAlphaBoolEnabled:bool=@n10, GravityOverride="0,0,0", ExternalForce="0,0,0") -> (Pose) {
    alphaInputType=Bool
    angularDampingOverride=1
    angularSpringConstant=20000
    bAngularSpring=true
    bOverrideAngularDamping=true
    bUseGravityOverride=true
    gravityScale=0.35
    simulationSpace=World
    alphaBoolBlend {blendOption=Linear}
    boundBone {boneName=dyn_Sleeve_under_l}
    componentAppliedLinearAccClamp=<3000 3000 3000>
    componentLinearAccScale=<0.3 0.3 0.3>
    componentLinearVelScale=<0.3 0.3 0.3>
    gravityOverride {z=-980}
    physicsBodyDefinitions [{
      boundBone {boneName=dyn_Sleeve_under_l}
      boxExtents=<20 30 30>
      constraintSetup {angularLimitsMax=<10 10 10>, angularLimitsMin=<-10 -10 -10>, angularTarget=<1 0 0>}
      localJointOffset=<0.3 0 -2>
     }]
   }
   n19 = "Constraint - MSTR_NeckZip (PPPPP)"!(
    ComponentPose=@n18
    Alpha=1
    ConstraintWeights_0=0.4204
    ConstraintWeights_1:float=0.2392
    ConstraintWeights_2:float=0.2072
    ConstraintWeights_3:float=0.1051
    ConstraintWeights_4:float=0.028
   ) -> (Pose) {
    alphaBoolBlend {blendOption=Linear}
    boneToModify {boneName=MSTR_NeckZip}
    constraintSetup [
     {targetBone {boneName=neck_02}}
     {targetBone {boneName=neck_01}}
     {targetBone {boneName=spine_05}}
     {targetBone {boneName=head}}
     {targetBone {boneName=spine_04}}
    ]
    constraintWeights [1,1,1,1,1]
   }
   n20 = Look At - Bone: AIM_NeckZip(ComponentPose=@n19, Alpha=1) {
    alphaBoolBlend {blendOption=Linear}
    boneToModify {boneName=AIM_NeckZip}
    lookAtLocation=<0 8.1>
    lookAtTarget {boneReference {boneName=spine_05}}
    lookAt_Axis {axis=<1 0>}
    lookUp_Axis {axis {x=0}}
   }
   n21 = Anim Dynamics - Bone: Dyn_NeckZip!(ComponentPose=@n20, Alpha=1, GravityOverride:Vector=@n1, ExternalForce="0,0,0") -> (Pose) {
    angularDampingOverride=1
    angularSpringConstant=500
    bAngularSpring=true
    bOverrideAngularDamping=true
    bUseGravityOverride=true
    gravityScale=0.35
    simulationSpace=RootRelative
    alphaBoolBlend {blendOption=Linear}
    boundBone {boneName=Dyn_NeckZip}
    componentAppliedLinearAccClamp=<3000 3000 3000>
    componentLinearAccScale=<0.3 0.3 0.3>
    componentLinearVelScale=<0.3 0.3 0.3>
    gravityOverride {z=-980}
    physicsBodyDefinitions [{
      boundBone {boneName=Dyn_NeckZip}
      boxExtents=<10 16 8>
      constraintSetup {angularLimitsMax=<2 20 20>, angularLimitsMin=<-2 -20 -20>, angularTarget=<0 0 0>}
      localJointOffset=<4 0 1>
     }]
    planarLimits [{drivingBone {boneName=AIM_NeckZip}, planeTransform {location=<0 0 0>, rotation=<rot 0 0 0>, scale=<0.1 0.1 0.1>}}]
   }
   n22 = Anim Dynamics - Bone: Dyn_NeckZip!(ComponentPose=@n21, bAlphaBoolEnabled:bool=Get bIsInFrontEnd(), GravityOverride=@n1, ExternalForce="0,0,0") -> (Pose) {
    alphaInputType=Bool
    angularDampingOverride=1
    angularSpringConstant=500
    bAngularSpring=true
    bOverrideAngularDamping=true
    bUseGravityOverride=true
    gravityScale=0.35
    simulationSpace=World
    alphaBoolBlend {blendOption=Linear}
    boundBone {boneName=Dyn_NeckZip}
    componentAppliedLinearAccClamp=<3000 3000 3000>
    componentLinearAccScale=<0.3 0.3 0.3>
    componentLinearVelScale=<0.3 0.3 0.3>
    gravityOverride {z=-980}
    physicsBodyDefinitions [{
      boundBone {boneName=Dyn_NeckZip}
      boxExtents=<10 16 8>
      constraintSetup {angularLimitsMax=<2 20 20>, angularLimitsMin=<-2 -20 -20>, angularTarget=<0 0 0>}
      localJointOffset=<4 0 1>
     }]
   }
   Component To Local(ComponentPose=@n22)
   Output Pose(Result=Copy Pose From Mesh() {bUseAttachedParent=true}) {layerGroup=None}
   n23 = Control Rig() {
    alphaBoolBlend {blendOption=Linear}
    controlRigAssetReference {blueprintRigClass=&/F_MED_UpbeatIguana_CtrlRig.F_MED_UpbeatIguana_CtrlRig_C}
   }
   n24 = "Constraint - base_03 (PPPPPPPP)"!(
    ComponentPose=Local To Component(LocalPose=@n23)
    Alpha=1
    ConstraintWeights_0=0.2517
    ConstraintWeights_1:float=0.2517
    ConstraintWeights_2:float=0.1533
    ConstraintWeights_3:float=0.1333
    ConstraintWeights_4:float=0.1046
    ConstraintWeights_5:float=0.0477
    ConstraintWeights_6:float=0.0359
    ConstraintWeights_7:float=0.0218
   ) -> (Pose) {
    alphaBoolBlend {blendOption=Linear}
    boneToModify {boneName=base_03}
    constraintSetup [
     {targetBone {boneName=clavicle_r}}
     {targetBone {boneName=deform_upperarm_bk_r}}
     {targetBone {boneName=upperarm_twist_01_r}}
     {targetBone {boneName=spine_05}}
     {targetBone {boneName=deform_upperarm_up_r}}
     {targetBone {boneName=neck_01}}
     {targetBone {boneName=spine_04}}
     {targetBone {boneName=deform_upperarm_fr_r}}
    ]
    constraintWeights [1,1,1,1,1,1,1,1]
   }
   n25 = "Constraint - base_04 (PPPPPPPP)"!(
    ComponentPose=@n24
    Alpha=1
    ConstraintWeights_0=0.3542
    ConstraintWeights_1:float=0.1637
    ConstraintWeights_2:float=0.1384
    ConstraintWeights_3:float=0.1356
    ConstraintWeights_4:float=0.1173
    ConstraintWeights_5:float=0.0443
    ConstraintWeights_6:float=0.0282
    ConstraintWeights_7:float=0.0183
   ) -> (Pose) {
    alphaBoolBlend {blendOption=Linear}
    boneToModify {boneName=base_04}
    constraintSetup [
     {targetBone {boneName=spine_04}}
     {targetBone {boneName=spine_02}}
     {targetBone {boneName=spine_05}}
     {targetBone {boneName=clavicle_l}}
     {targetBone {boneName=spine_03}}
     {targetBone {boneName=spine_01}}
     {targetBone {boneName=upperarm_twist_01_l}}
     {targetBone {boneName=clavicle_r}}
    ]
    constraintWeights [1,1,1,1,1,1,1,1]
   }
   n26 = "Transform (Modify) Bone - Bone: Strap_04"(ComponentPose=@n25, Alpha=1, Translation="0,-4,0", Rotation="0,0,0", Scale="1,1,1") {
    translationMode=BMM_Additive
    translationSpace=BCS_BoneSpace
    alphaBoolBlend {blendOption=Linear}
    boneToModify {boneName=Strap_04}
    translation {y=-4}
   }
   n27 = "Constraint - base_ZipPurse_01 (P)"!(ComponentPose=@n26, Alpha=1, ConstraintWeights_0=1, ConstraintWeights_1:float=0) -> (Pose) {
    alphaBoolBlend {blendOption=Linear}
    boneToModify {boneName=base_ZipPurse_01}
    constraintSetup [{targetBone {boneName=dyn_ZipPurse_01}}]
    constraintWeights [1,0]
   }
   n28 = "Transform (Modify) Bone - Bone: deform_upperarm_up_base_l"!(ComponentPose=@n27, bAlphaBoolEnabled:bool=Get HangGliding() -> (HangGliding:bool), Translation="0,0,-8", Rotation="0,0,0", Scale="1,1,1") -> (Pose) {
    alphaInputType=Bool
    translationMode=BMM_Additive
    translationSpace=BCS_BoneSpace
    alphaBoolBlend {blendInTime=0.25, blendOption=Linear, blendOutTime=0.25}
    boneToModify {boneName=deform_upperarm_up_base_l}
    translation {z=-8}
   }
   n29 = "Transform (Modify) Bone - Bone: Strap_01"!(
    ComponentPose=@n28
    bAlphaBoolEnabled:bool=Get bIsSwimSprinting() -> (bIsSwimSprinting:bool)
    Translation="5,1,5"
    Rotation="0,0,0"
    Scale="1,1,1"
   ) -> (Pose) {
    alphaInputType=Bool
    translationMode=BMM_Additive
    translationSpace=BCS_BoneSpace
    alphaBoolBlend {blendInTime=0.25, blendOption=Linear, blendOutTime=0.25}
    boneToModify {boneName=Strap_01}
    translation=<5 1 5>
   }
   n30 = "Transform (Modify) Bone - Bone: Strap_04"!(ComponentPose=@n29, bAlphaBoolEnabled:bool=Get bIsSurfaceSwimming(), Translation="3,-3,0", Rotation="0,0,0", Scale="1,1,1") -> (Pose) {
    alphaInputType=Bool
    translationMode=BMM_Additive
    translationSpace=BCS_BoneSpace
    alphaBoolBlend {blendInTime=0.25, blendOption=Linear, blendOutTime=0.25}
    boneToModify {boneName=Strap_04}
    translation=<3 -3>
   }
   n31 = "Transform (Modify) Bone - Bone: attach_backpack"(ComponentPose=@n30, Alpha=Get isGlamgunner() -> (isGlamgunner:double), Translation="0,-1.283,0", Rotation="0,0,0", Scale="1,1,1") {
    translationMode=BMM_Additive
    translationSpace=BCS_BoneSpace
    alphaBoolBlend {blendOption=Linear}
    boneToModify {boneName=attach_backpack}
    translation {y=-1.283}
   }
   Save cached pose 'Cache01'(Pose=Component To Local(ComponentPose=@n31)) {cacheName=Cache01}
   n32 = Get Override World Gravity()
   n33 = RigidBody!(
    ComponentPose=Local To Component(LocalPose=Use cached pose 'Cache01'())
    Alpha
    OverrideWorldGravity:Vector=@n32
    SimSpaceSettings:Sim Space Settings Structure=(WorldAlpha=0.8,VelocityScaleZ=1,DampingAlpha=1,MaxLinearVelocity=500,MaxAngularVelocity=3,MaxLinearAcceleration=700,MaxAngularAcceleration=1000,ExternalLinearDragV=(X=0.2,Y=0.2,Z=0.1),ExternalLinearVelocity=(X=0,Y=0,Z=0),ExternalAngularVelocity=(X=0,Y=0,Z=0))
   ) -> (Pose) {
    bOverrideWorldGravity=true
    overridePhysicsAsset=&/F_MED_UpbeatIguana_Physics
    simulationSpace=BaseBoneSpace
    alphaBoolBlend {blendOption=Linear}
    baseBoneRef {boneName=spine_01}
    componentAppliedLinearAccClamp=<3000 3000 1000>
    componentLinearAccScale=<0.25 0.25 0.1>
    componentLinearVelScale=<0.25 0.25 0.1>
    overrideWorldGravity {z=-980}
    simSpaceSettings {
     maxAngularAcceleration=1000
     maxAngularVelocity=3
     maxLinearAcceleration=700
     maxLinearVelocity=500
     worldAlpha=0.8
     externalLinearDragV=<0.2 0.2 0.1>
    }
   }
   n34 = Anim Dynamics - Bone: dyn_purse_03(ComponentPose=@n33, Alpha=0.7, ExternalForce="0,0,0") {
    angularDampingOverride=1
    angularSpringConstant=20000
    bAngularSpring=true
    bOverrideAngularDamping=true
    bUseGravityOverride=true
    gravityScale=0.35
    simulationSpace=RootRelative
    alphaBoolBlend {blendOption=Linear}
    boundBone {boneName=dyn_purse_03}
    componentLinearAccScale=<1 1 1>
    componentLinearVelScale=<1 1 1>
    gravityOverride {z=-980}
    physicsBodyDefinitions [{
      boundBone {boneName=dyn_purse_03}
      boxExtents=<20 30 30>
      constraintSetup {angularLimitsMax=<15 15 15>, angularLimitsMin=<-15 -15 -15>, angularTarget=<1 0 0>}
      localJointOffset=<4 0 0>
     }]
   }
   n35 = Anim Dynamics - Bone: dyn_ZipPurse_01(ComponentPose=@n34, Alpha=1, ExternalForce=@n32) {
    angularDampingOverride=1
    angularSpringConstant=20000
    bAngularSpring=true
    bOverrideAngularDamping=true
    bUseGravityOverride=true
    gravityScale=0.35
    simulationSpace=RootRelative
    alphaBoolBlend {blendOption=Linear}
    boundBone {boneName=dyn_ZipPurse_01}
    componentLinearAccScale=<1 1 1>
    componentLinearVelScale=<1 1 1>
    gravityOverride {z=-980}
    physicsBodyDefinitions [{
      boundBone {boneName=dyn_ZipPurse_01}
      boxExtents=<10 16 8>
      constraintSetup {angularLimitsMax=<0 45 45>, angularLimitsMin=<0 -45 -45>, angularTarget=<0 0 0>}
      localJointOffset=<2.8 0 0>
     }]
   }
   n36 = Anim Dynamics - Bone: dyn_ZipPurse_01!(ComponentPose=@n35, bAlphaBoolEnabled:bool=Get bIsInFrontEnd(), ExternalForce="0,0,0") -> (Pose) {
    alphaInputType=Bool
    angularDampingOverride=1
    angularSpringConstant=20000
    bAngularSpring=true
    bOverrideAngularDamping=true
    bUseGravityOverride=true
    gravityScale=0.35
    simulationSpace=World
    alphaBoolBlend {blendOption=Linear}
    boundBone {boneName=dyn_ZipPurse_01}
    componentLinearAccScale=<1 1 1>
    componentLinearVelScale=<1 1 1>
    gravityOverride {z=-980}
    physicsBodyDefinitions [{
      boundBone {boneName=dyn_ZipPurse_01}
      boxExtents=<10 16 8>
      constraintSetup {angularLimitsMax=<0 45 45>, angularLimitsMin=<0 -45 -45>, angularTarget=<0 0 0>}
      localJointOffset=<2.8 0 0>
     }]
   }
   Save cached pose 'Cache02'(Pose=Component To Local(ComponentPose=@n36)) {cacheName=Cache02}
  ]}
 EventGraph {nodes [
   Event BlueprintInitializeAnimation()
   > Set SkelMesh(SkelMesh=GetOwnerSkeletalMeshComponent())
   n1 = TryGetPawnOwner()
   > Is Valid(InputObject=@n1)
   > n2 = Cast To FortPlayerPawnAthena(Object=@n1)
   > Set PlayerPawn(PlayerPawn:Fort Player Pawn Athena Object Reference=@n2) -> (Output_Get:Fort Player Pawn Athena Object Reference)
   Event OnParachuteClosed()
   > Set HangGliding(HangGliding:bool=false) -> (Output_Get:bool)
   Event OnParachuteOpened()
   > Switch on EGliderType(Selection=Get PlayerGliderType() -> (PlayerGliderType=HangGlider))
   > Set HangGliding(HangGliding=true)
   Event OnCharacterCustomizationLoaded()
   > n3 = Set SkelMesh(SkelMesh=GetPartSkeletalMeshComponent())
   > n4 = Get Dynamic Material Instance(SkeletalMeshComponent=@n3)
   > Set Mat Instance Dynamic(Mat Instance Dynamic=@n4)
   > n5 = GetOrCreateDynamicOverlayMaterial(MeshComponent=@n3, MaterialName=None)
   > Set Overlay Mat Instance Dynamic(Overlay Mat Instance Dynamic=@n5)
   > n6 = Get Dynamic Material Instance(SkeletalMeshComponent=@n3, ElementIndex=1)
   > Set Mat Instance Dynamic 1(Mat Instance Dynamic 1:Material Instance Dynamic Object Reference=@n6) -> (Output_Get:Material Instance Dynamic Object Reference)
   > n7 = GetOrCreateDynamicOverlayMaterial(MeshComponent=@n3, MaterialName=None)
   > Set Overlay Mat Instance Dynamic 1(Overlay Mat Instance Dynamic 1:Material Instance Dynamic Object Reference=@n7) -> (Output_Get:Material Instance Dynamic Object Reference)
   n8 = Sequence!(execute) -> (then_0, then_1, then_2:Exec)
   > n9 = Gravity Override State Function DEPRECATED(
    CharacterPartAnimInstance=Self-Reference()
    GameTimeInSeconds=0
    GravityOverrideParams=Get Gravity Override Params()
    RBAN Params=Get RBAN Params()
   )
   > Set Override World Gravity(Override World Gravity:Vector=@n9) -> (Output_Get:Vector)
   n10 = Get bIsJumping()
   > n11 = Branch(Condition=AND Boolean(A=NOT Boolean(A=@n10), B=Get bIsSkyDiving()))
   > Set Delta Sleeve Off(Delta Sleeve Off:double=0) -> (Output_Get:double)
   Set Delta Sleeve Off(execute=@n11.else, Delta Sleeve Off=1)
   Is Valid(exec=@n8.then_1, InputObject=Get PlayerPawn() -> (PlayerPawn:Fort Player Pawn Athena Object Reference))
   > n12 = Branch(Condition=Get bIsInFrontEnd())
   n13 = Set WPO Globals Function(
    execute=@n12.else
    Material=Get Mat Instance Dynamic()
    Overlay Material=Get Overlay Mat Instance Dynamic()
    Skeletal Mesh=Get SkelMesh()
    "Local Noise Transform  Location"=Get Local Noise Transform Location() -> (Local Noise Transform Location:Vector)
    Local Noise Offset=Get Local Noise Offset()
   )
   n14 = Event BlueprintUpdateAnimation()
   Set WPO Regional Function!(
    execute=@n13.then
    Material:Material Instance Dynamic Object Reference=Get Mat Instance Dynamic()
    Overlay Material:Material Instance Dynamic Object Reference=Get Overlay Mat Instance Dynamic()
    Skeletal Mesh:Skeletal Mesh Component Object Reference=Get SkelMesh()
    Velocity Transform:Name=foot_r
    Velocity Interp Speed:double=Get Velocity Interp Speed()
    Collision Transform:Name=calf_r
    Collision Transform Axis:EAxis Enum
    Invert Collision Transform Axis:bool=false
    Delta Time X:double=@n14.DeltaTimeX
    Previous Velocity Transform Location:"Vector (by ref)"=Get R1 Velocity Transform Location()
    Previous Velocity Transform Velocity:"Vector (by ref)"=Get R1 Velocity Transform Velocity()
    Region:int=1
   ) -> (then)
   > Set WPO Regional Function(
    Material=Get Mat Instance Dynamic()
    Overlay Material=Get Overlay Mat Instance Dynamic()
    Skeletal Mesh=Get SkelMesh()
    Velocity Transform=foot_l
    Velocity Interp Speed=Get Velocity Interp Speed()
    Collision Transform=calf_l
    Invert Collision Transform Axis=true
    Delta Time X=@n14.DeltaTimeX
    Previous Velocity Transform Location=Get R2 Velocity Transform Location()
    Previous Velocity Transform Velocity=Get R2 Velocity Transform Velocity()
    Region=2
   )
   > Set WPO Regional Function(
    Material=Get Mat Instance Dynamic 1() -> (Mat Instance Dynamic 1:Material Instance Dynamic Object Reference)
    Overlay Material=Get Overlay Mat Instance Dynamic 1() -> (Overlay Mat Instance Dynamic 1:Material Instance Dynamic Object Reference)
    Skeletal Mesh=Get SkelMesh()
    Velocity Transform=hand_l
    Velocity Interp Speed=Get Velocity Interp Speed()
    Collision Transform=upperarm_l
    Invert Collision Transform Axis=true
    Delta Time X=@n14.DeltaTimeX
    Previous Velocity Transform Location=Get R2 Velocity Transform Location()
    Previous Velocity Transform Velocity=Get R2 Velocity Transform Velocity()
    Region=2
   )
   > Set WPO Regional Function(
    Material=Get Mat Instance Dynamic 1()
    Overlay Material=Get Overlay Mat Instance Dynamic 1()
    Skeletal Mesh=Get SkelMesh()
    Velocity Transform=hand_r
    Velocity Interp Speed=Get Velocity Interp Speed()
    Collision Transform=upperarm_r
    Invert Collision Transform Axis=false
    Delta Time X=@n14.DeltaTimeX
    Previous Velocity Transform Location=Get R1 Velocity Transform Location()
    Previous Velocity Transform Velocity=Get R1 Velocity Transform Velocity()
    Region=2
   )
   > n15 = Branch(Condition=OR Boolean(A=Get bIsSkydiveFloating(), B=Get bIsParachuteOpened()))
   > Set WPO Pho Wind Intensity(WPO Pho Wind Intensity=0.8)
   > n16 = Set WPO Pho Turbulence(WPO Pho Turbulence:double=1) -> (Output_Get:double)
   SetScalarParameterValue(
    execute=[@n18.then, @n16.then]
    self=[Get Mat Instance Dynamic 1(), Get Overlay Mat Instance Dynamic 1(), Get Mat Instance Dynamic(), Get Overlay Mat Instance Dynamic()]
    ParameterName=ClothSim_WindStrength
    Value=Get WPO Pho Wind Intensity()
   )
   n17 = Get WPO Pho Turbulence() -> (WPO Pho Turbulence:double)
   > SetScalarParameterValue(
    self=[Get Mat Instance Dynamic 1(), Get Overlay Mat Instance Dynamic 1(), Get Mat Instance Dynamic(), Get Overlay Mat Instance Dynamic()]
    ParameterName=ClothSim_R1_TurbulenceStrength
    Value=@n17
   )
   > SetScalarParameterValue(
    self=[Get Overlay Mat Instance Dynamic(), Get Mat Instance Dynamic(), Get Mat Instance Dynamic 1(), Get Overlay Mat Instance Dynamic 1()]
    ParameterName=ClothSim_R2_TurbulenceStrength
    Value=@n17
   )
   Set WPO Pho Wind Intensity(execute=@n15.else, WPO Pho Wind Intensity=0)
   > n18 = Set WPO Pho Turbulence(WPO Pho Turbulence=0)
   n19 = Cast To FortPlayerPawn(execute=@n8.then_2, Object=TryGetPawnOwner())
   n20 = GetDisplayName(Object=GetSkeletalMeshForPartType(self=@n19, PartType=Face))
   > Set isGlamgunner(isGlamgunner:double="To Float (Boolean)"(InBool=MatchesWildcard(SourceString=@n20, Wildcard=*F_MED_DyedDuelist_FaceAcc*, SearchCase=CaseSensitive))) -> (Output_Get:double)
  ]}
]
variables [
 Body_Mesh {type=SkeletalMeshComponent}
 BodyMesh {type=SkeletalMeshComponent}
 Delta Sleeve Off {type=number}
 Gravity Override Params {type=GravityOverrideParamsStruct}
 HangGliding {type=boolean}
 isGlamgunner {type=number}
 Local Noise Offset {type=Vector}
 Local Noise Transform Location {type=Vector}
 Mat Instance Dynamic {type=MaterialInstanceDynamic}
 Mat Instance Dynamic 1 {type=MaterialInstanceDynamic}
 Overlay Mat Instance Dynamic {type=MaterialInstanceDynamic}
 Overlay Mat Instance Dynamic 1 {type=MaterialInstanceDynamic}
 Override World Gravity {type=Vector}
 PawnOwner {type=PlayerPawn_Athena_C}
 PlayerPawn {type=FortPlayerPawnAthena}
 R1 Velocity Transform Location {type=Vector}
 R1 Velocity Transform Velocity {type=Vector}
 R2 Velocity Transform Location {type=Vector}
 R2 Velocity Transform Velocity {type=Vector}
 RBAN Params {
  type=RBANParamsStruct
  default {
   dBNO {simSpaceSettings {
     maxAngularAcceleration=1000
     maxAngularVelocity=3
     maxLinearAcceleration=300
     maxLinearVelocity=700
     velocityScaleZ=0
     externalLinearDragV=<0 0.5 0>
    }}
   dBNO_Carried {simSpaceSettings {
     maxAngularAcceleration=1000
     maxAngularVelocity=3
     maxLinearAcceleration=300
     maxLinearVelocity=700
     velocityScaleZ=0
     externalLinearDragV=<0 0.5 0>
    }}
   emoteOrMelee {simSpaceSettings {
     maxAngularAcceleration=1000
     maxAngularVelocity=3
     maxLinearAcceleration=300
     maxLinearVelocity=700
     velocityScaleZ=0
     externalLinearDragV=<0 0.5 0>
    }}
   front_End {simSpaceSettings {
     maxAngularAcceleration=1000
     maxAngularVelocity=3
     maxLinearAcceleration=300
     maxLinearVelocity=700
     velocityScaleZ=0
     externalLinearDragV=<0 0.5 0>
    }}
   onGround_Crouch_Moving {simSpaceSettings {
     maxAngularAcceleration=1000
     maxAngularVelocity=3
     maxLinearAcceleration=300
     maxLinearVelocity=700
     velocityScaleZ=0
     externalLinearDragV=<0 0.5 0>
    }}
   onGround_Falling {simSpaceSettings {
     maxAngularAcceleration=1000
     maxAngularVelocity=3
     maxLinearAcceleration=300
     maxLinearVelocity=700
     velocityScaleZ=0
     externalLinearDragV=<0 0.5 0>
    }}
   onGround_Moving {simSpaceSettings {
     maxAngularAcceleration=1000
     maxAngularVelocity=3
     maxLinearAcceleration=300
     maxLinearVelocity=700
     velocityScaleZ=0
     externalLinearDragV=<0 0.5 0>
    }}
   onGround_Standing {simSpaceSettings {
     maxAngularAcceleration=1000
     maxAngularVelocity=3
     maxLinearAcceleration=300
     maxLinearVelocity=700
     velocityScaleZ=0
     externalLinearDragV=<0 0.5 0>
    }}
   rail_Grinding {simSpaceSettings {
     maxAngularAcceleration=1000
     maxAngularVelocity=3
     maxLinearAcceleration=300
     maxLinearVelocity=700
     velocityScaleZ=0
     worldAlpha=0.1
     externalLinearDragV=<0 0.5 0>
    }}
   riding_Moving {simSpaceSettings {
     maxAngularAcceleration=1000
     maxAngularVelocity=3
     maxLinearAcceleration=300
     maxLinearVelocity=700
     velocityScaleZ=0
     externalLinearDragV=<0 0.5 0>
    }}
   riding_Moving_Targeting {simSpaceSettings {
     maxAngularAcceleration=1000
     maxAngularVelocity=3
     maxLinearAcceleration=300
     maxLinearVelocity=700
     velocityScaleZ=0
     externalLinearDragV=<0 0.5 0>
    }}
   riding_Standing {simSpaceSettings {
     maxAngularAcceleration=1000
     maxAngularVelocity=3
     maxLinearAcceleration=300
     maxLinearVelocity=700
     velocityScaleZ=0
     externalLinearDragV=<0 0.5 0>
    }}
   skydive_DivingDown {simSpaceSettings {
     maxAngularAcceleration=1000
     maxAngularVelocity=3
     maxLinearAcceleration=300
     maxLinearVelocity=700
     velocityScaleZ=0
     externalLinearDragV=<0 0.5 0>
    }}
   skydive_FloatingDown {
    componentLinearAccScale=<0 0.5 0>
    componentLinearVelScale=<0 0.5 0>
    simSpaceSettings {
     maxAngularAcceleration=1000
     maxAngularVelocity=3
     maxLinearAcceleration=300
     maxLinearVelocity=700
     velocityScaleZ=0
     externalLinearDragV=<0 0.5 0>
    }
   }
   skydive_GoingUp {simSpaceSettings {
     maxAngularAcceleration=1000
     maxAngularVelocity=3
     maxLinearAcceleration=300
     maxLinearVelocity=700
     velocityScaleZ=0
     externalLinearDragV=<0 0.5 0>
    }}
   skydive_Parachuting {
    componentLinearAccScale=<0 0.1 0.3>
    componentLinearVelScale=<0 0.1 0.3>
    simSpaceSettings {
     maxAngularAcceleration=1000
     maxAngularVelocity=3
     maxLinearAcceleration=300
     maxLinearVelocity=700
     velocityScaleZ=0
     externalLinearDragV=<0 0.5 0>
    }
   }
   skydive_Surfing {simSpaceSettings {
     maxAngularAcceleration=1000
     maxAngularVelocity=3
     maxLinearAcceleration=300
     maxLinearVelocity=700
     velocityScaleZ=0
     externalLinearDragV=<0 0.5 0>
    }}
   swimming {simSpaceSettings {
     maxAngularAcceleration=1000
     maxAngularVelocity=3
     maxLinearAcceleration=300
     maxLinearVelocity=700
     velocityScaleZ=0
     externalLinearDragV=<0 0.5 0>
    }}
   tactical_Sprint {simSpaceSettings {
     maxAngularAcceleration=1000
     maxAngularVelocity=3
     maxLinearAcceleration=300
     maxLinearVelocity=700
     velocityScaleZ=0
     externalLinearDragV=<0 0.5 0>
    }}
   ziplining {simSpaceSettings {
     maxAngularAcceleration=1000
     maxAngularVelocity=3
     maxLinearAcceleration=300
     maxLinearVelocity=700
     velocityScaleZ=0
     externalLinearDragV=<0 0.5 0>
    }}
  }
 }
 SkelMesh {type=SkeletalMeshComponent}
 SleeveAlpha {type=number}
 TailMesh {type=SkeletalMeshComponent}
 Velocity Interp Speed {type=number, default=35}
 WPO Pho Turbulence {type=number}
 WPO Pho Wind Intensity {type=number}
]
